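# Material function for the Superlevels player skin that builds tiled cylindrical UVs from pre-skinned local position, with separate handling for the arms.
# Arm space comes from rotation transforms mirrored by X sign, blended in through shoulder, axis and height masks, with static switches for arm-only or original position.
# Radial angle and bounds-normalized height form the UVs.
# DDX/DDY derive from a camera-vector-offset position and are output alongside arm masks and a horizontal gradient.
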MaterialFunction /BRCosmetics/Characters/Player/Male/Medium/Bodies/M_MED_CoolSlice_GolfVolley/Skins/Superlevels/Material/MF_CylindricalProjUV_Complex_Superlevel
graph {nodes [
  n1 = NamedRerouteUsage()
  n2 = NamedRerouteUsage()
  Add(A=@n1, B=@n2)
  n3 = NamedRerouteUsage()
  n4 = VectorParameter() {group=UVProjection_Combine, parameterName=ChestPos_Pivot, defaultValue=<rgba 0 0 150 1>}
  Divide(
   A=Length(None=Subtract(A=@n3, B=@n4.RGB)) {is=class MaterialExpressionLength}
   B=ScalarParameter() {defaultValue=60, group=UVProjection_Combine, parameterName=ChestRaidus}
  )
  FunctionOutput(None=NamedRerouteUsage()) {outputName=UVs}
  FunctionOutput(None=NamedRerouteUsage()) {outputName=DDX, sortPriority=1}
  FunctionOutput(None=NamedRerouteUsage()) {outputName=DDY, sortPriority=2}
  FunctionOutput(None=NamedRerouteUsage()) {outputName=ArmMask, sortPriority=10}
  FunctionOutput(None=NamedRerouteUsage()) {outputName=ArmGradient, sortPriority=20}
  FunctionOutput(None=NamedRerouteUsage()) {outputName=Horiztonal, sortPriority=30}
  n5 = ScalarParameter() {defaultValue=0.05, group=UVProjection_Rotation, parameterName="RotationDedgreeZ(Radius)"}
  n6 = Cosine(None=@n5)
  n7 = Sine(None=@n5)
  n8 = Constant()
  n9 = AppendVector(A=AppendVector(A=@n6, B=@n7), B=@n8)
  n10 = Multiply(A=@n7) {constB=-1}
  MaterialFunctionCall(
   VectorToTransform=NamedRerouteUsage()
   BasisX=LinearInterpolate(A=Multiply(A=@n9) {constB=-1}, B=@n9, Alpha=NamedRerouteUsage())
   BasisY=AppendVector(A=AppendVector(A=@n10, B=@n6), B=@n8)
   BasisZ=Constant3Vector() {constant=<rgba 0 0 1 0>}
   BasisPosition=VectorParameter() {group=UVProjection_Rotation, parameterName=BasisPositionY, defaultValue=<rgba 0 0 0 1>}.RGB
  ) {materialFunction=&Transform3x3Matrix}
  n11 = NamedRerouteUsage()
  n12 = CameraVectorWS()
  n13 = Multiply(
   A=Transform(None=Multiply(A=@n12) {constB=-1}) {transformSourceType=TRANSFORMSOURCE_World, transformType=TRANSFORM_Local}
   B=NamedRerouteUsage()
  ) {constB=100}
  n14 = Divide(A=NamedRerouteDeclaration(None=Add(A=@n11, B=@n13)) {name=SL_PreskinnedLocalPos, nodeColor=<rgba 0 1 0.0859 1>}, B=NamedRerouteUsage())
  n15 = Normalize(VectorInput=Transform(None=Constant3Vector() {constant=<rgba 1 0 0 0>}) {transformSourceType=TRANSFORMSOURCE_View})
  NamedRerouteDeclaration(None=DDX(Value=DotProduct(A=@n14, B=@n15))) {name=DDX, nodeColor=<rgba 0 1 0.9062 1>}
  NamedRerouteDeclaration(None=DDY(Value=ComponentMask(None=@n14) {b=true})) {name=DDY, nodeColor=<rgba 1 0 0.9609 1>}
  n16 = ScalarParameter() {defaultValue=50, group=UVProjection_Combine, parameterName=ArmAxisGradient}
  n17 = NamedRerouteUsage()
  n18 = ComponentMask(None=@n17) {b=true}
  n19 = Add(
   A=Step(Y=@n16, X=@n18)
   B=SphereMask(A=@n16, B=@n18, Radius=ScalarParameter() {defaultValue=5, group=UVProjection_Combine, parameterName=ArmblendRadius}) {attenuationRadius=10, hardnessPercent=0}
  )
  NamedRerouteDeclaration(None=OneMinus(None=Saturate(None=@n19))) {name=ArmGradient, nodeColor=<rgba 0.0547 1 0 1>}
  n20 = FunctionInput(Preview=VertexInterpolator(VS=LocalPosition() {localOrigin=InstancePreSkinning}.XYZ).PS) {bUsePreviewValueAsDefault=true, inputName=LocalPosition, sortPriority=-5}
  n21 = Abs(None=ComponentMask(None=@n20) {r=true})
  NamedRerouteDeclaration(None=@n21) {name=Gradient_Horiztonal, nodeColor=<rgba 0.6875 1 0 1>}
  n22 = ScalarParameter() {defaultValue=0.1, group=UVProjection_Rotation, parameterName="RotationDedgreeY(Radius)"}
  NamedRerouteDeclaration(None=@n22) {name=RotateDegree, nodeColor=<rgba 1 0 0.6328 1>}
  n23 = NamedRerouteUsage()
  n24 = Cosine(None=@n22)
  n25 = Constant()
  n26 = Sine(None=@n22)
  n27 = NamedRerouteDeclaration(None=AppendVector(A=AppendVector(A=@n24, B=@n25), B=Multiply(A=@n26) {constB=-1})) {name=RotateX01, nodeColor=<rgba 1 0.1641 0 1>}
  n28 = Saturate(None=Sign(None=ComponentMask(None=@n23) {r=true}))
  n29 = LinearInterpolate(A=Multiply(A=@n27) {constB=-1}, B=@n27, Alpha=@n28)
  n30 = Constant3Vector() {constant=<rgba 0 1 0 0>}
  n31 = Constant()
  n32 = NamedRerouteDeclaration(None=AppendVector(A=AppendVector(A=@n26, B=@n31), B=@n24)) {name=RotateZ01, nodeColor=<rgba 1 0.1641 0 1>}
  n33 = VectorParameter() {desc=Default Value, group=UVProjection_Rotation, parameterName=BasisPositionRight, defaultValue=<rgba -122.3188 -22.2314 60 1>}
  n34 = FunctionInput(Preview=@n33.RGB) {bUsePreviewValueAsDefault=true, inputName=BasisPosition, sortPriority=-4}
  NamedRerouteDeclaration(None=NamedRerouteDeclaration(None=MaterialFunctionCall(VectorToTransform=@n23, BasisX=@n29, BasisY=@n30, BasisZ=@n32, BasisPosition=@n34) {materialFunction=&Transform3x3Matrix}.Transformed Vector) {name=VectorArm01, nodeColor=<rgba 1 0.5391 0 1>}) {name=Arm, nodeColor=<rgba 1 0 0.1875 1>}
  n35 = MaterialFunctionCall(Vector or UVs=NamedRerouteUsage()) {materialFunction=&VectorToRadialValue}
  n36 = NamedRerouteUsage()
  n37 = NamedRerouteUsage()
  n38 = NamedRerouteUsage()
  n39 = PreSkinnedLocalBounds()
  n40 = SmoothStep(
   Min=ScalarParameter() {defaultValue=0.27, group=UVProjection_Combine, parameterName=ShoulderRatio}
   Max=ScalarParameter() {defaultValue=1, group=UVProjection_Combine, parameterName=ShoulderEndRatio}
   Value=Divide(A=@n21, B=ComponentMask(None=@n39.Half Extents) {r=true})
  ) {constMax=0.5}
  n41 = ScalarParameter() {defaultValue=3, group=UVProjection_Combine, parameterName=ShoulderPower}
  n42 = FunctionInput(Preview=ScalarParameter() {defaultValue=50, group=UVProjection_Combine, parameterName=ArmAxisEnd}) {bUsePreviewValueAsDefault=true, inputName=ArmAxisEnd, inputType=FunctionInput_Scalar, sortPriority=-3}
  n43 = ComponentMask(None=@n17) {b=true}
  n44 = ScalarParameter() {defaultValue=5, group=UVProjection_Combine, parameterName=ArmblendRadius}
  n45 = Saturate(None=Add(A=Step(Y=@n42, X=@n43), B=SphereMask(A=@n42, B=@n43, Radius=@n44) {attenuationRadius=10, hardnessPercent=0}))
  n46 = Multiply(A=Multiply(A=Power(Base=@n40, Exp=@n41), B=OneMinus(None=@n45)), B=NamedRerouteUsage())
  n47 = StaticSwitchParameter(
   True=NamedRerouteDeclaration(None=@n17) {name=Arm_Coord, nodeColor=<rgba 0.7578 1 0 1>}
   False=LinearInterpolate(A=@n1, B=@n2, Alpha=NamedRerouteDeclaration(None=@n46) {name=ArmMask, nodeColor=<rgba 0.3828 0 1 1>})
  ) {group=UVProjection_Combine, parameterName=UseArmCoordOnly}
  n48 = ComponentMask(None=NamedRerouteDeclaration(None=StaticSwitchParameter(True=@n38, False=@n47) {group=UVProjection_Combine, parameterName=UseOG_POS}) {name=BlendedPOS, nodeColor=<rgba 0 1 0.4844 1>}) {b=true}
  n49 = ComponentMask(None=@n39.Min) {b=true}
  n50 = ComponentMask(None=@n39.Extents) {b=true}
  n51 = Saturate(None=Divide(A=Subtract(A=@n48, B=@n49), B=NamedRerouteDeclaration(None=@n50) {name=SL_PreskinnedLocalBoundsZ, nodeColor=<rgba 0 1 0.3203 1>}))
  n52 = AppendVector(
   A=Multiply(A=Multiply(A=@n35.Vector Converted to Angle) {constB=-1}, B=ComponentMask(None=@n36) {r=true}) {constB=2}
   B=Multiply(A=ComponentMask(None=@n37) {g=true}, B=OneMinus(None=@n51))
  )
  n53 = FunctionInput(Preview=ScalarParameter() {defaultValue=1, group=UVProjection_Combine, parameterName=ArmDoubleTiling}) {bUsePreviewValueAsDefault=true, inputName=ArmDoubleTiling, inputType=FunctionInput_Scalar, sortPriority=20}
  NamedRerouteDeclaration(None=LinearInterpolate(A=@n52, B=Multiply(A=@n52, B=@n53), Alpha=NamedRerouteUsage())) {name=UVs, nodeColor=<rgba 0.6875 0 1 1>}
  NamedRerouteDeclaration(None=@n28) {name=XMask, nodeColor=<rgba 0.8047 0 1 1>}
  n54 = FunctionInput() {bUsePreviewValueAsDefault=true, inputName=Tile, inputType=FunctionInput_Vector2, previewValue=<1 1 0 1>}
  NamedRerouteDeclaration(None=@n54) {name=Tile, nodeColor=<rgba 0.4531 0 1 1>}
  n55 = FunctionInput() {bUsePreviewValueAsDefault=true, inputName=DistortionAdjustment, inputType=FunctionInput_Scalar, sortPriority=1, previewValue=<-5 0 0 1>}
  NamedRerouteDeclaration(None=@n55) {name=DistortionAdjustment, nodeColor=<rgba 0.4531 0 1 1>}
  NamedRerouteDeclaration(None=SmoothStep(
    Min=ScalarParameter() {defaultValue=60, group=UVProjection_Combine, parameterName=ArmEndHeight}
    Max=ScalarParameter() {defaultValue=65, group=UVProjection_Combine, parameterName=ArmStartHeight}
    Value=ComponentMask(None=@n20) {b=true}
   )) {name=ArmHeightMask, nodeColor=<rgba 0 0.3437 1 1>}
  NamedRerouteDeclaration(None=@n20) {name=OG_POS, nodeColor=<rgba 1 0.2109 0 1>}
  NamedRerouteUsage()
  Saturate(None=Sign(None=ComponentMask(None=@n23) {r=true}))
 ]}
