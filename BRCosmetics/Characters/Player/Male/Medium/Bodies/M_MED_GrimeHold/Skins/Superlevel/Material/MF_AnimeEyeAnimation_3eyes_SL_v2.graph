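# Material function that animates three anime-style eyes for the GrimeHold Superlevel skin, blending eye and lid colors into lit and shaded inputs.
# UV tiles separate right, middle and left eyes, each with procedural lids driven by blink, squint and lid-down pose scalars.
# Look poses shift a pupil map or procedural pupil, with optional FOV-compensated camera-distance scaling.
# Lid gradients produce masks, distance-field-style outlines and overlap roughness, with static switches disabling features.
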MaterialFunction /BRCosmetics/Characters/Player/Male/Medium/Bodies/M_MED_GrimeHold/Skins/Superlevel/Material/MF_AnimeEyeAnimation_3eyes_SL_v2
graph {nodes [
  n1 = NamedRerouteUsage()
  Add(A=@n1, B=NamedRerouteUsage())
  n2 = NamedRerouteUsage()
  Add(A=@n2, B=NamedRerouteUsage())
  n3 = NamedRerouteUsage()
  Add(A=@n3, B=NamedRerouteUsage())
  n4 = VectorParameter() {group=01_EyeAnimation_Parameters, parameterName=PupilOffset_L, defaultValue=<rgba 0 0 0 1>}
  n5 = VectorParameter() {group=01_EyeAnimation_Parameters, parameterName=PupilOffset_M, defaultValue=<rgba 0 0 0 1>}
  n6 = Add(
   A=Multiply(A=ComponentMask(None=@n4.RGB) {g=true, r=true}, B=NamedRerouteUsage())
   B=Multiply(A=ComponentMask(None=@n5.RGB) {g=true, r=true}, B=NamedRerouteUsage())
  )
  n7 = Multiply(
   A=ComponentMask(None=VectorParameter() {group=01_EyeAnimation_Parameters, parameterName=PupilOffset_R, defaultValue=<rgba 0 0 0 1>}.RGB) {g=true, r=true}
   B=NamedRerouteUsage()
  )
  n8 = NamedRerouteUsage()
  n9 = MaterialFunctionCall(Float2=Add(A=Add(A=@n6, B=@n7), B=Frac(None=@n8))) {materialFunction=&BreakOutFloat2Components}
  n10 = Add(A=@n9.G) {constB=-0.5}
  n11 = ScalarParameter() {defaultValue=2, desc=horizontal UV ratio, group=01_EyeAnimation_Parameters, parameterName=PupilMap_UV_Ratio, sliderMax=2, sortPriority=5}
  n12 = MaterialFunctionCall(X=@n9.R, Y=Add(A=Multiply(A=@n10, B=@n11)) {constB=0.5}) {materialFunction=&MakeFloat2}
  Add(A=@n12.Result, B=Multiply(A=NamedRerouteUsage(), B=Constant2Vector() {r=-1})) {desc=move right eye to UV space 0-1}
  n13 = NamedRerouteUsage()
  n14 = Distance(A=Multiply(A=Add(A=@n13) {constB=-0.5}) {constB=2}, B=Constant())
  n15 = ScalarParameter() {defaultValue=0.4, group=01_EyeAnimation_Parameters, parameterName=Pupil_DF_LineThickness, sliderMax=1}
  n16 = Saturate(None=SphereMask(
    A=Saturate(None=@n14)
    B=Constant() {r=1}
    Radius=OneMinus(None=@n15)
    Hardness=ScalarParameter() {defaultValue=0.9, group=01_EyeAnimation_Parameters, parameterName=Pupil_DF_LineHardness, sliderMax=1}
   ) {attenuationRadius=0.1})
  FunctionOutput(None=StaticSwitchParameter(True=OneMinus(None=@n16), False=Constant()) {group=01_EyeAnimation_Parameters, parameterName=UseEyeAnimation, sortPriority=-1}) {outputName=EyeMask, sortPriority=3}
  n17 = NamedRerouteDeclaration(None=FunctionInput() {inputName=Color Lit, sortPriority=1}) {name=InColorLit, nodeColor=<rgba 1 0 0.0234 1>}
  n18 = NamedRerouteUsage()
  n19 = NamedRerouteUsage()
  FunctionOutput(None=StaticSwitchParameter(True=LinearInterpolate(A=@n17, B=@n18, Alpha=@n19), False=@n17) {group=01_EyeAnimation_Parameters, parameterName=UseEyeAnimation, sortPriority=-1}) {outputName=Lit, sortPriority=1}
  n20 = NamedRerouteDeclaration(None=FunctionInput() {inputName=Color Shaded, sortPriority=2}) {name=InColorShaded, nodeColor=<rgba 0 0.3437 1 1>}
  n21 = NamedRerouteUsage()
  n22 = NamedRerouteUsage()
  FunctionOutput(None=StaticSwitchParameter(True=LinearInterpolate(A=@n20, B=@n21, Alpha=@n22), False=@n20) {group=01_EyeAnimation_Parameters, parameterName=UseEyeAnimation, sortPriority=-1}) {outputName=Shaded, sortPriority=2}
  FunctionOutput(None=NamedRerouteUsage()) {outputName=UpperLidMask, sortPriority=10}
  FunctionOutput(None=NamedRerouteUsage()) {outputName=LowerLidMask, sortPriority=11}
  FunctionOutput(None=NamedRerouteUsage()) {outputName=DF_Lids, sortPriority=12}
  n23 = Add(
   A=ScalarParameter() {group=01_EyeAnimation_Poses, parameterName=look_left_pose, sliderMax=1, sortPriority=11}
   B=Multiply(A=ScalarParameter() {group=01_EyeAnimation_Poses, parameterName=look_right_pose, sliderMax=1, sortPriority=10}) {constB=-1}
  )
  n24 = Add(
   A=ScalarParameter() {group=01_EyeAnimation_Poses, parameterName=look_up_pose, sliderMax=1, sortPriority=12}
   B=Multiply(A=ScalarParameter() {group=01_EyeAnimation_Poses, parameterName=look_down_pose, sliderMax=1, sortPriority=13}) {constB=-1}
  )
  n25 = AppendVector(
   A=Multiply(
    A=Clamp(None=@n23) {minDefault=-1}
    B=ScalarParameter() {defaultValue=0.2, group=01_EyeAnimation_Parameters, parameterName=EyeHorizontal_Threshold, sliderMax=1}
   )
   B=Multiply(
    A=Clamp(None=@n24) {minDefault=-1}
    B=ScalarParameter() {defaultValue=0.2, group=01_EyeAnimation_Parameters, parameterName=EyeVertical_Threshold, sliderMax=1}
   )
  )
  n26 = ScalarParameter() {defaultValue=0.5, group=01_EyeAnimation_Parameters, parameterName=PupilMap_Scale, sliderMax=2, sortPriority=5}
  n27 = ScalarParameter() {defaultValue=500, group=01_EyeAnimation_Parameters, parameterName=PupilDistanceScaling_FadeLength, sliderMax=2, sortPriority=5}
  n28 = ScalarParameter() {defaultValue=500, group=01_EyeAnimation_Parameters, parameterName=PupilDistanceScaling_StartDistance, sliderMax=2, sortPriority=5}
  n29 = InverseLinearInterpolate(Value=MaterialFunctionCall() {materialFunction=/Engine/Functions/Engine_MaterialFunctions02/Utility/FOV}.FOV angle in degrees) {constA=45, constB=90}
  n30 = LinearInterpolate(
   A=@n26
   B=ScalarParameter() {defaultValue=0.5, group=01_EyeAnimation_Parameters, parameterName=PupilMap_ScaleFar, sliderMax=2, sortPriority=5}
   Alpha=Multiply(
    A=MaterialFunctionCall(Fade Length=@n27, Fade Offset=@n28) {materialFunction=&CameraDepthFade}.Result
    B=Saturate(None=@n29) {desc=FOV compensation}
   )
  )
  n31 = MaterialFunctionCall(Float2=MaterialFunctionCall(
    UVs=Add(A=@n12.Result, B=@n25)
    Texture Scale=StaticSwitchParameter(True=@n30, False=@n26) {group=01_EyeAnimation_Parameters, parameterName=UsePupilDistanceScaling, sortPriority=4}
   ) {materialFunction=&ScaleUVsByCenter}.UVs) {materialFunction=&BreakOutFloat2Components}
  n32 = Subtract(A=@n31.R) {constB=0.5}
  n33 = MaterialFunctionCall(X=Add(A=Multiply(A=@n32) {constB=2}) {constB=0.5}, Y=@n31.G) {materialFunction=&MakeFloat2}
  n34 = MaterialFunctionCall(Float2=@n33.Result) {materialFunction=&BreakOutFloat2Components}
  LinearInterpolate(
   A=@n33.Result
   B=MaterialFunctionCall(X=@n34.G, Y=@n34.G) {materialFunction=&MakeFloat2}.Result
   Alpha=NamedRerouteUsage()
  ) {desc=rotate right eye UVs}
  n35 = FunctionInput(Preview=StaticBool() {value=true}) {
   bUsePreviewValueAsDefault=true
   description="if true (default), uses the second UV channel for eye mapping. If false, uses the first UV channel."
   inputName=Use UV2
   inputType=FunctionInput_StaticBool
   sortPriority=15
   previewValue=<1 0 0 1>
  }
  NamedRerouteDeclaration(None=StaticSwitch(True=TextureCoordinate() {coordinateIndex=1}, False=TextureCoordinate(), Value=@n35)) {name=EyeUVs, nodeColor=<rgba 0.2188 0 1 1>}
  n36 = VectorParameter() {group=01_EyeAnimation_Parameters, parameterName=EyeLid_Color_Lit, sortPriority=3, defaultValue=<rgba 1 0.6445 0.0452 1>}
  n37 = TextureSampleParameter2D(UVs=@n13) {
   group=01_EyeAnimation_Parameters
   parameterName=PupilMap
   samplerSource=SSM_Clamp_WorldGroupSettings
   sortPriority=1
   texture=/Game/Characters/Player/Female/Medium/Bodies/F_MED_Genius/Textures/T_F_MED_Genius_Eye_D
  }
  n38 = StaticSwitchParameter(True=@n37.RGB, False=@n16) {group=01_EyeAnimation_Parameters, parameterName=UsePupilMap, sortPriority=0}
  n39 = NamedRerouteUsage()
  n40 = Max(A=NamedRerouteUsage(), B=NamedRerouteUsage())
  n41 = StaticSwitch(
   True=NamedRerouteUsage()
   False=Saturate(None=Subtract(A=@n39, B=@n40)) {desc=EyeBallMask}
   Value=NamedRerouteDeclaration(None=StaticBoolParameter() {group=01_EyeAnimation_Parameters, parameterName=useDisableEyeLids, sortPriority=1}) {name=DisableLids, nodeColor=<rgba 0 1 0.9589 1>}
  )
  n42 = LinearInterpolate(A=@n36.RGB, B=@n38, Alpha=@n41)
  NamedRerouteDeclaration(None=@n42) {name=Debug, nodeColor=<rgba 1 0.5625 0 1>}
  n43 = LinearInterpolate(
   A=ScalarParameter() {group=01_EyeAnimation_Parameters, parameterName=OverlapRoughness_Eye_Right, sliderMax=1, sortPriority=20}
   B=ScalarParameter() {group=01_EyeAnimation_Parameters, parameterName=OverlapRoughness_Eye_Left, sliderMax=1, sortPriority=20}
   Alpha=NamedRerouteUsage()
  )
  n44 = ScalarParameter() {defaultValue=0.4, group=01_EyeAnimation_Parameters, parameterName=OverlapRoughness_EyeLid_Right, sliderMax=1, sortPriority=20}
  n45 = ScalarParameter() {defaultValue=0.5, group=01_EyeAnimation_Parameters, parameterName=OverlapRoughness_EyeLid_Left, sliderMax=1, sortPriority=20}
  n46 = LinearInterpolate(A=@n44, B=@n45, Alpha=NamedRerouteUsage())
  n47 = ScalarParameter() {defaultValue=0.8, group=01_EyeAnimation_Parameters, parameterName=OverlapRoughness_EyeLid_Upper, sliderMax=1, sortPriority=21}
  n48 = NamedRerouteUsage()
  n49 = ScalarParameter() {defaultValue=8, group=01_EyeAnimation_Parameters, parameterName=OverlapRoughness_Gradient_Scale, sliderMax=16, sortPriority=25}
  n50 = OneMinus(None=Saturate(None=Multiply(A=@n48, B=@n49) {constB=8}))
  n51 = NamedRerouteUsage()
  n52 = ScalarParameter() {defaultValue=0.6, group=01_EyeAnimation_Parameters, parameterName=OverlapRoughness_EyeLid_Lower, sliderMax=1, sortPriority=22}
  n53 = NamedRerouteUsage()
  n54 = OneMinus(None=Saturate(None=Multiply(A=@n53, B=@n49) {constB=8}))
  n55 = NamedRerouteUsage()
  n56 = Add(
   A=Multiply(A=LinearInterpolate(A=@n46, B=@n47, Alpha=@n50) {constB=0}, B=@n51)
   B=Multiply(A=LinearInterpolate(A=@n46, B=@n52, Alpha=@n54) {constB=0}, B=@n55)
  )
  n57 = Max(A=@n51, B=@n55)
  NamedRerouteDeclaration(None=StaticSwitch(True=@n43, False=LinearInterpolate(A=@n43, B=@n56, Alpha=@n57), Value=NamedRerouteUsage())) {name=OverlapRoughness, nodeColor=<rgba 0.0955 0.2229 1 1>}
  n58 = VectorParameter() {group=01_EyeAnimation_Parameters, parameterName=EyeLid_Color_DF_Line, sortPriority=3, defaultValue=<rgba 0 0 0 1>}
  n59 = NamedRerouteUsage()
  NamedRerouteDeclaration(None=StaticSwitchParameter(True=@n42, False=LinearInterpolate(A=@n58.RGB, B=@n42, Alpha=@n59)) {group=01_EyeAnimation_Parameters, parameterName=UseOverlapOutlines, sortPriority=0}) {name=EyeColor_Lit, nodeColor=<rgba 1 0.5322 0.1287 1>}
  n60 = Floor(None=Add(A=NamedRerouteUsage()) {constB=0})
  n61 = MaterialFunctionCall(Float2=@n60) {materialFunction=&BreakOutFloat2Components}
  n62 = Max(A=Add(A=@n61.R, B=Constant() {r=-2}), B=@n61.G) {constB=0}
  n63 = MaterialFunctionCall(Float2=Multiply(A=@n60) {constB=-1}) {materialFunction=&BreakOutFloat2Components}
  n64 = Max(A=@n63.R, B=@n63.R) {constB=0}
  NamedRerouteDeclaration(None=OneMinus(None=Max(A=@n62, B=@n64) {constB=0})) {name=Mask_EyesUDIM, nodeColor=<rgba 1 0.2578 0 1>}
  n65 = MaterialFunctionCall(Float2=NamedRerouteUsage()) {materialFunction=&BreakOutFloat2Components}
  n66 = Floor(None=@n65.R) {desc=LeftEye}
  n67 = Saturate(None=Floor(None=Subtract(A=@n65.R)))
  NamedRerouteDeclaration(None=Subtract(A=Saturate(None=@n66), B=@n67)) {name=Mask_EyeMiddle, nodeColor=<rgba 0 0.2969 1 1>}
  n68 = Sine(None=MaterialFunctionCall(Float2=NamedRerouteUsage()) {materialFunction=&BreakOutFloat2Components}.R)
  n69 = OneMinus(None=Multiply(A=@n68, B=@n68))
  n70 = ScalarParameter() {group=01_EyeAnimation_Parameters, parameterName=UpperLid_Close_Curveture_R, sliderMax=1}
  n71 = Saturate(None=Cosine(None=NamedRerouteUsage()))
  NamedRerouteDeclaration(None=Multiply(A=Multiply(A=@n69, B=@n70) {constB=0.15}, B=OneMinus(None=@n71))) {name=R_UL_Close_Curveture, nodeColor=<rgba 0 1 0.2969 1>}
  n72 = ScalarParameter() {group=01_EyeAnimation_Poses, parameterName=R_upperLidDown_pose, sliderMax=1, sortPriority=4}
  n73 = ScalarParameter() {group=01_EyeAnimation_Poses, parameterName=R_blink_pose, sliderMax=1, sortPriority=2}
  n74 = ScalarParameter() {group=01_EyeAnimation_Poses, parameterName=R_squint_inner_pose, sliderMax=1, sortPriority=6}
  n75 = ScalarParameter() {defaultValue=0.35, group=01_EyeAnimation_Parameters, parameterName=Squint_Eyelid_Influence, sliderMax=1, sortPriority=31}
  n76 = Max(A=@n73, B=Multiply(A=@n74, B=@n75) {constB=0.5})
  n77 = NamedRerouteUsage()
  n78 = Saturate(None=Power(Base=Max(A=@n72, B=@n76), Exp=@n77))
  n79 = Add(
   A=Multiply(A=@n1) {constB=-1}
   B=LinearInterpolate(
    A=NamedRerouteUsage()
    B=NamedRerouteUsage()
    Alpha=NamedRerouteDeclaration(None=@n78) {name=R_upperLid_close_pose, nodeColor=<rgba 0.125 1 0 1>}
   ) {constB=0.6}
  ) {constB=0}
  NamedRerouteDeclaration(None=Abs(None=Add(A=@n79) {constB=-0.5})) {name=UR_Grad, nodeColor=<rgba 0.15 1 0 1>}
  n80 = ScalarParameter() {group=01_EyeAnimation_Parameters, parameterName=UpperLid_Close_Curveture_L, sliderMax=1}
  n81 = Saturate(None=Cosine(None=NamedRerouteUsage()))
  NamedRerouteDeclaration(None=Multiply(A=Multiply(A=@n69, B=@n80) {constB=0.15}, B=OneMinus(None=@n81))) {name=L_UL_Close_Curveture, nodeColor=<rgba 0 1 0.2969 1>}
  n82 = Power(Base=Max(A=@n73, B=@n74), Exp=@n77)
  n83 = Add(A=@n1, B=LinearInterpolate(A=NamedRerouteUsage(), B=NamedRerouteUsage(), Alpha=Saturate(None=@n82)) {constB=0.4}) {constB=0}
  NamedRerouteDeclaration(None=Abs(None=Add(A=@n83) {constB=-0.5})) {name=LR_Grad, nodeColor=<rgba 0 1 0.0156 1>}
  n84 = ScalarParameter() {group=01_EyeAnimation_Poses, parameterName=L_upperLidDown_pose, sliderMax=1, sortPriority=3}
  n85 = ScalarParameter() {group=01_EyeAnimation_Poses, parameterName=L_squint_inner_pose, sliderMax=1, sortPriority=5}
  n86 = ScalarParameter() {group=01_EyeAnimation_Poses, parameterName=L_blink_pose, sliderMax=1, sortPriority=1}
  n87 = Max(A=Multiply(A=@n85, B=@n75) {constB=0.5}, B=@n86)
  n88 = NamedRerouteUsage()
  n89 = Saturate(None=Power(Base=Max(A=@n84, B=@n87), Exp=@n88))
  n90 = Add(
   A=Multiply(A=@n2) {constB=-1}
   B=LinearInterpolate(
    A=NamedRerouteUsage()
    B=NamedRerouteUsage()
    Alpha=NamedRerouteDeclaration(None=@n89) {name=L_upperLid_close_pose, nodeColor=<rgba 0.125 1 0 1>}
   ) {constB=0.6}
  ) {constB=0}
  NamedRerouteDeclaration(None=Abs(None=Add(A=@n90) {constB=-0.5})) {name=UL_Grad, nodeColor=<rgba 1 0.5333 0 1>}
  n91 = Power(Base=Max(A=@n86, B=@n85), Exp=@n88)
  n92 = Add(A=@n2, B=LinearInterpolate(A=NamedRerouteUsage(), B=NamedRerouteUsage(), Alpha=Saturate(None=@n91)) {constB=0.4}) {constB=0}
  NamedRerouteDeclaration(None=Abs(None=Add(A=@n92) {constB=-0.5})) {name=LL_Grad, nodeColor=<rgba 0.8952 1 0 1>}
  n93 = Min(A=Min(A=NamedRerouteUsage(), B=NamedRerouteUsage()), B=NamedRerouteUsage())
  n94 = Min(A=Min(A=NamedRerouteUsage(), B=NamedRerouteUsage()), B=NamedRerouteUsage())
  n95 = Min(A=NamedRerouteUsage(), B=NamedRerouteUsage())
  n96 = NamedRerouteUsage()
  n97 = ScalarParameter() {defaultValue=60, group=01_EyeAnimation_Parameters, parameterName=DF_Lid_Gradient_Scale_Center, sliderMax=32, sortPriority=25}
  n98 = ScalarParameter() {defaultValue=1, group=01_EyeAnimation_Parameters, parameterName=DF_Lid_Gradient_Scale_Edge, sliderMax=32, sortPriority=25}
  n99 = NamedRerouteUsage()
  n100 = ScalarParameter() {defaultValue=1, group=01_EyeAnimation_Parameters, parameterName=DF_Lid_Gradient_Scale_Power, sliderMax=32, sortPriority=25}
  n101 = Power(Base=Distance(A=Frac(None=@n99), B=Constant() {r=0.5}), Exp=@n100) {constExponent=0.5}
  n102 = ScalarParameter() {defaultValue=16, group=01_EyeAnimation_Parameters, parameterName=DF_Lid_Gradient_Scale, sliderMax=32, sortPriority=25}
  n103 = Multiply(
   A=Max(A=Max(A=@n93, B=@n94), B=Min(A=@n95, B=@n96))
   B=StaticSwitchParameter(True=LinearInterpolate(A=@n97, B=@n98, Alpha=@n101) {desc=Compensate for thicker line in eyeball center due to UV}, False=@n102) {group=01_EyeAnimation_Parameters, parameterName=Use2LidGradientScales, sortPriority=25}
  ) {constB=16}
  n104 = ScalarParameter() {defaultValue=0.2, group=01_EyeAnimation_Parameters, parameterName=Lid_DF_LineThickness, sliderMax=1}
  NamedRerouteDeclaration(None=SphereMask(
    A=Saturate(None=@n103)
    B=Constant() {r=1}
    Radius=OneMinus(None=@n104)
    Hardness=ScalarParameter() {defaultValue=0.9, group=01_EyeAnimation_Parameters, parameterName=Lid_DF_LineHardness, sliderMax=1}
   ) {attenuationRadius=0.6, hardnessPercent=95}) {name=DF_Lids, nodeColor=<rgba 1 0.3047 0 1>}
  NamedRerouteDeclaration(None=@n33.Result) {name=EyeCoords, nodeColor=<rgba 0 1 0.6016 1>}
  n105 = VectorParameter() {group=01_EyeAnimation_Parameters, parameterName=EyeLid_Color_Shaded, sortPriority=3, defaultValue=<rgba 0.5972 0.3813 0.0343 1>}
  n106 = LinearInterpolate(A=@n105.RGB, B=@n38, Alpha=@n41)
  NamedRerouteDeclaration(None=StaticSwitchParameter(True=@n106, False=LinearInterpolate(A=@n58.RGB, B=@n106, Alpha=@n59)) {group=01_EyeAnimation_Parameters, parameterName=UseOverlapOutlines, sortPriority=0}) {name=EyeColor_Shaded, nodeColor=<rgba 1 0.5322 0.1287 1>}
  NamedRerouteDeclaration(None=ScalarParameter() {group=01_EyeAnimation_Parameters, parameterName=UpperLid_Min, sliderMax=1, sortPriority=28}) {name=UpperLid_Min, nodeColor=<rgba 0.9688 0 1 1>}
  n107 = OneMinus(None=@n66)
  n108 = Ceil(None=@n65.G)
  n109 = NamedRerouteUsage()
  n110 = FunctionInput() {bUsePreviewValueAsDefault=true, inputName=Mask, inputType=FunctionInput_Scalar, sortPriority=20, previewValue=<1 0 0 1>}
  n111 = NamedRerouteDeclaration(None=Multiply(A=Min(A=@n108, B=@n109), B=@n110)) {name=Mask_Eyes, nodeColor=<rgba 1 0 0.8906 1>}
  NamedRerouteDeclaration(None=Saturate(None=Multiply(A=@n107, B=@n111) {desc=RightEye})) {name=Mask_EyeRight, nodeColor=<rgba 1 0 0.0152 1>}
  NamedRerouteDeclaration(None=ScalarParameter() {defaultValue=0.6, group=01_EyeAnimation_Parameters, parameterName=UpperLid_Max, sliderMax=1, sortPriority=29}) {name=UpperLid_Max, nodeColor=<rgba 0 1 0.6484 1>}
  NamedRerouteDeclaration(None=ScalarParameter() {group=01_EyeAnimation_Parameters, parameterName=LowerLid_Min, sliderMax=1, sortPriority=30}) {name=LowerLid_Min, nodeColor=<rgba 1 0.3281 0 1>}
  NamedRerouteDeclaration(None=ScalarParameter() {defaultValue=0.4, group=01_EyeAnimation_Parameters, parameterName=LowerLid_Max, sliderMax=1, sortPriority=31}) {name=LowerLid_Max, nodeColor=<rgba 0.0078 0 1 1>}
  n112 = ScalarParameter() {defaultValue=1, group=01_EyeAnimation_Parameters, parameterName=AnimationThreshold_Eyes, sliderMax=8, sortPriority=30}
  NamedRerouteDeclaration(None=@n112) {name=EyesThreshold, nodeColor=<rgba 1 0 0.5625 1>}
  n113 = ScalarParameter() {group=01_EyeAnimation_Poses, parameterName=M_upperLidDown_pose, sliderMax=1, sortPriority=4}
  n114 = ScalarParameter() {group=01_EyeAnimation_Poses, parameterName=M_blink_pose, sliderMax=1, sortPriority=2}
  n115 = ScalarParameter() {group=01_EyeAnimation_Poses, parameterName=M_squint_inner_pose, sliderMax=1, sortPriority=6}
  n116 = Max(A=@n114, B=Multiply(A=@n115, B=@n75) {constB=0.5})
  n117 = NamedRerouteUsage()
  n118 = Saturate(None=Power(Base=Max(A=@n113, B=@n116), Exp=@n117))
  n119 = Add(
   A=Multiply(A=@n3) {constB=-1}
   B=LinearInterpolate(
    A=NamedRerouteUsage()
    B=NamedRerouteUsage()
    Alpha=NamedRerouteDeclaration(None=@n118) {name=M_upperLid_close_pose, nodeColor=<rgba 0.125 1 0 1>}
   ) {constB=0.6}
  ) {constB=0}
  NamedRerouteDeclaration(None=Abs(None=Add(A=@n119) {constB=-0.5})) {name=UM_Grad, nodeColor=<rgba 0.15 1 0 1>}
  n120 = Power(Base=Max(A=@n114, B=@n115), Exp=@n117)
  n121 = Add(A=@n3, B=LinearInterpolate(A=NamedRerouteUsage(), B=NamedRerouteUsage(), Alpha=Saturate(None=@n120)) {constB=0.4}) {constB=0}
  NamedRerouteDeclaration(None=Abs(None=Add(A=@n121) {constB=-0.5})) {name=LM_Grad, nodeColor=<rgba 0 1 0.0156 1>}
  NamedRerouteDeclaration(None=Multiply(A=Add(A=@n72, B=@n84)) {constB=0.5}) {name=M_upperLidDown_pose, nodeColor=<rgba 0 0.3437 1 1>}
  NamedRerouteDeclaration(None=Multiply(A=Add(A=@n74, B=@n85)) {constB=0.5}) {name=M_squint_inner_pose, nodeColor=<rgba 0 0.3438 1 1>}
  NamedRerouteDeclaration(None=@n67) {name=Mask_EyeLeft, nodeColor=<rgba 0.0293 0.3125 0.0353 1>}
  n122 = NamedRerouteUsage()
  n123 = Saturate(None=Multiply(A=Round(None=@n79), B=@n122)) {desc=UR}
  n124 = NamedRerouteUsage()
  n125 = Saturate(None=Multiply(A=Round(None=@n119), B=@n124)) {desc=UR}
  n126 = NamedRerouteUsage()
  n127 = Multiply(A=Round(None=@n90), B=@n126)
  NamedRerouteDeclaration(None=Max(A=Max(A=@n123, B=@n125) {desc=UpperLidMask}, B=Saturate(None=@n127) {desc=UL})) {name=Mask_UpperLid, nodeColor=<rgba 0 0.5781 1 1>}
  n128 = NamedRerouteUsage()
  NamedRerouteDeclaration(None=MaterialFunctionCall(Float2=Add(A=@n128) {constB=-0.5, desc=LeftEyeUVs}) {materialFunction=&BreakOutFloat2Components}.G) {name=EyeUVs.G, nodeColor=<rgba 1 0.4922 0 1>}
  n129 = Saturate(None=Multiply(A=Round(None=@n83), B=@n122)) {desc=LR}
  n130 = Saturate(None=Multiply(A=Round(None=@n121), B=@n124)) {desc=LR}
  n131 = Multiply(A=Round(None=@n92), B=@n126)
  NamedRerouteDeclaration(None=Max(A=Max(A=@n129, B=@n130) {desc=LowerLidMask}, B=Saturate(None=@n131) {desc=LL}) {desc=LowerLidMask}) {name=Mask_LowerLid, nodeColor=<rgba 1 0 0.5391 1>}
  n132 = Multiply(A=NamedRerouteUsage())
  n133 = Multiply(A=Multiply(A=@n132) {constB=-1}, B=NamedRerouteUsage())
  n134 = Add(A=NamedRerouteUsage()) {constB=-0.5}
  NamedRerouteDeclaration(None=Saturate(None=Add(A=@n133, B=@n134) {constB=0.1})) {name=UpperLidsGradients_Overlay, nodeColor=<rgba 1 0.9375 0 1>}
  n135 = Add(A=NamedRerouteUsage()) {constB=-0.5}
  NamedRerouteDeclaration(None=Saturate(None=Add(A=@n132, B=@n135))) {name=LowerLidsGradients_Overlay, nodeColor=<rgba 0.4075 1 0.2387 1>}
  NamedRerouteUsage()
  NamedRerouteUsage()
  NamedRerouteUsage()
 ]}
