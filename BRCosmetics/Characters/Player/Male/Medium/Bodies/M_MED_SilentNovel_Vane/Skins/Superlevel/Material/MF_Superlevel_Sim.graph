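# Material function for the SilentNovel Vane Superlevel skin, producing animated green bubbling, dripping "nuclear" surfaces for anime lit/shaded character shading.
# Pre-skinned-position projections drive distorted bubbles and gradient patterns; arm-aware cylindrical UVs with explicit derivatives drive drip masks and normals.
# Texture masks, camera-relative fake lights and Fresnel terms composite tinted transmission, highlights and fake AO.
# Outputs colors, normals, sheen and custom light vector, with optional animated eyes and eye-based temporal responsiveness.
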MaterialFunction /BRCosmetics/Characters/Player/Male/Medium/Bodies/M_MED_SilentNovel_Vane/Skins/Superlevel/Material/MF_Superlevel_Sim
graph {nodes [
  n1 = NamedRerouteUsage()
  n2 = NamedRerouteUsage()
  Add(A=LinearInterpolate(A=Constant(), B=Sign(None=@n1), Alpha=ScalarParameter() {group=UVProjection_RE, parameterName=DisableArm}), B=@n2)
  n3 = NamedRerouteUsage()
  n4 = NamedRerouteUsage()
  n5 = Saturate(None=ScalarParameter() {defaultValue=1, group=010_Superlevel_Control, parameterName=CustomLightBlend})
  FunctionOutput(None=Normalize(VectorInput=LinearInterpolate(A=@n3, B=@n4, Alpha=@n5))) {outputName=CustomeLightVector, sortPriority=-1}
  FunctionOutput(None=NamedRerouteUsage()) {outputName=OverrideNormal_WS}
  FunctionOutput(None=NamedRerouteUsage()) {outputName=Normal_Map, sortPriority=1}
  FunctionOutput(None=NamedRerouteUsage()) {outputName=Sheen_Mask, sortPriority=2}
  FunctionOutput(None=NamedRerouteUsage()) {outputName=Color_Lit, sortPriority=3}
  FunctionOutput(None=NamedRerouteUsage()) {outputName=Color_Shaded, sortPriority=4}
  n6 = VectorParameter() {group=010_Superlevel, parameterName=DistortAmountRE, defaultValue=<rgba 1 2 0 0.05>}
  n7 = Add(A=NamedRerouteUsage(), B=Multiply(A=NamedRerouteUsage(), B=ComponentMask(None=@n6.RGB) {g=true, r=true}))
  n8 = TextureObjectParameter() {
   group=010_Superlevel
   parameterName=NuclearPattern
   samplerType=SAMPLERTYPE_LinearColor
   texture=/BRCosmetics/Characters/Player/Male/Medium/Bodies/M_MED_SilentNovel_Vane/Skins/Superlevel/Texture/T_Superlevel_NuclearMask
  }
  n9 = NamedRerouteUsage()
  n10 = NamedRerouteUsage()
  n11 = TextureSample!(UVs=@n7, Tex=@n8, "DDX(UVs)"=@n9, "DDY(UVs)"=@n10, Apply View MipBias) {
   mipValueMode=TMVM_Derivative
   samplerSource=SSM_Wrap_WorldGroupSettings
   samplerType=SAMPLERTYPE_LinearColor
   texture=/BRCosmetics/Characters/Player/Male/Medium/Bodies/M_MED_SilentNovel_Vane/Skins/Superlevel/Texture/T_Superlevel_Bubbles
  }
  n12 = DeriveNormalZ(InXY=Subtract(A=ComponentMask(None=@n11.RGB) {g=true, r=true}) {constB=0.5})
  MaterialFunctionCall(Normal=@n12, Flatness=Constant() {r=-10}) {materialFunction=&FlattenNormal}
  n13 = NamedRerouteUsage()
  n14 = NamedRerouteUsage()
  n15 = NamedRerouteUsage()
  n16 = NamedRerouteUsage()
  n17 = Add(A=LinearInterpolate(A=NamedRerouteUsage(), B=@n16, Alpha=NamedRerouteUsage()), B=Multiply(A=@n16, B=NamedRerouteUsage()))
  n18 = NamedRerouteUsage()
  n19 = Multiply(
   A=LinearInterpolate(A=LinearInterpolate(A=@n13, B=@n14, Alpha=@n15), B=Add(A=@n17, B=@n18), Alpha=NamedRerouteUsage())
   B=LinearInterpolate(Alpha=NamedRerouteUsage()) {constA=1, constB=0.5}
  )
  n20 = VectorParameter() {group=010_Superlevel, parameterName=SaturationColor, defaultValue=<rgba 0.0895 0.3448 0 1>}
  n21 = Multiply(A=NamedRerouteUsage(), B=Multiply(A=@n20.RGB, B=@n20.RGB))
  n22 = Subtract(A=NamedRerouteUsage(), B=NamedRerouteUsage())
  NamedRerouteDeclaration(None=LinearInterpolate(A=Add(A=@n19, B=@n21), Alpha=Saturate(None=@n22)) {constB=0}) {name=LitFinal, nodeColor=<rgba 1 0 0.0469 1>}
  n23 = LocalPosition() {localOrigin=InstancePreSkinning}
  n24 = VectorParameter() {group=UVProjection_RE, parameterName=CenterPivot_RE}
  n25 = VectorParameter() {desc=Default Value, group=UVProjection_RE, parameterName=BasisPositionRight_RE, defaultValue=<rgba -122.3188 -22.2314 60 1>}
  n26 = MaterialFunctionCall!(
   LocalPosition=VertexInterpolator(VS=Subtract(A=@n23.XYZ, B=@n24.RGB)).PS
   BasisPosition=@n25.RGB
   ArmAxisEnd=ScalarParameter() {defaultValue=50, group=UVProjection_RE, parameterName=ArmAxisEnd_RE}
   Tile=ComponentMask(None=VectorParameter() {group=UVProjection_RE, parameterName=CylinderUV_Tile, defaultValue=<rgba 2 2 2 0>}.RGB) {g=true, r=true}
   DistortionAdjustment=Constant()
   ArmDoubleTiling=ScalarParameter() {defaultValue=1, group=UVProjection_RE, parameterName=ArmDoubleTiling_RE}
  ) {materialFunction=/BRCosmetics/Characters/Player/Male/Medium/Bodies/M_MED_CoolSlice_GolfVolley/Skins/Superlevels/Material/MF_CylindricalProjUV_Complex_Superlevel}
  NamedRerouteDeclaration(None=@n26.ArmMask) {name=ArmMask, nodeColor=<rgba 0 0.9531 1 1>}
  n27 = TextureSampleParameter2D() {
   group=010_Superlevel
   parameterName=SuperlevelMask
   samplerType=SAMPLERTYPE_Masks
   texture=/BRCosmetics/Characters/Player/Male/Medium/Bodies/M_MED_SilentNovel_Vane/Textures/T_SilentNovel_Vane_Body_FX
  }
  NamedRerouteDeclaration(None=@n27.R) {name=GreenArea, nodeColor=<rgba 1 0 0 1>}
  n28 = VertexInterpolator(VS=LocalPosition() {localOrigin=InstancePreSkinning}.XYZ)
  n29 = ComponentMask(None=@n28.PS) {b=true, r=true}
  n30 = ScalarParameter() {defaultValue=70, group=010_Superlevel, parameterName=BubbleUnit}
  n31 = ComponentMask(None=@n28.PS) {b=true, r=true}
  n32 = ScalarParameter() {defaultValue=100, group=010_Superlevel, parameterName=BubbleDistortUnit}
  n33 = ComponentMask(None=VectorParameter() {group=010_Superlevel, parameterName=DistortSpeed, defaultValue=<rgba 0 -0.1 0 1>}.RGB) {g=true, r=true}
  n34 = TextureObjectParameter() {group=010_Superlevel, parameterName=DistortPattern, samplerType=SAMPLERTYPE_Normal, texture=/Game/Global/Textures/Default/Noise/T_Bubble_B_N}
  n35 = TextureSample!(UVs=Panner(Coordinate=Divide(A=@n31, B=@n32), Speed=@n33) {speedY=-0.1}, Tex=@n34, Apply View MipBias) {samplerSource=SSM_Wrap_WorldGroupSettings, samplerType=SAMPLERTYPE_Normal, texture=&T_Bubble_B_N}
  n36 = ComponentMask(None=@n28.PS) {b=true, g=true}
  n37 = TextureSample(UVs=Panner(Coordinate=Divide(A=@n36, B=@n32), Speed=@n33) {speedY=-0.1}, Tex=@n34) {samplerSource=SSM_Wrap_WorldGroupSettings, samplerType=SAMPLERTYPE_Normal, texture=&T_Bubble_B_N}
  n38 = ScalarParameter() {defaultValue=0.2, group=010_Superlevel, parameterName=FrontDistortionAmount}
  n39 = Abs(None=VertexInterpolator(VS=PreSkinnedNormal()).PS)
  n40 = Constant3Vector() {constant=<rgba 0 1 0 0>}
  n41 = OneMinus(None=Saturate(None=DotProduct(A=@n39, B=@n40)))
  n42 = Power(Base=@n41)
  n43 = Multiply(
   A=LinearInterpolate(A=ComponentMask(None=@n35.RGB) {g=true, r=true}, B=ComponentMask(None=@n37.RGB) {g=true, r=true}, Alpha=NamedRerouteUsage())
   B=Multiply(
    A=ScalarParameter() {defaultValue=0.05, group=010_Superlevel, parameterName=DistortIntensity}
    B=LinearInterpolate(A=@n38, Alpha=@n42) {constA=0.1}
   )
  ) {constB=0.1}
  NamedRerouteDeclaration(None=Add(A=Divide(A=@n29, B=@n30), B=@n43)) {name=ProceduralUV, nodeColor=<rgba 1 0.9375 0 1>}
  n44 = Normalize(VectorInput=VectorParameter() {group=010_Superlevel, parameterName=ScreenSpaceLightVector02, defaultValue=<rgba -1 -1 0.5 1>}.RGB)
  n45 = Constant3Vector() {constant=<rgba 0 1 0 0>}
  n46 = Normalize(VectorInput=WorldPosition() {worldPositionShaderOffset=WPT_CameraRelative}.XYZ)
  n47 = Normalize(VectorInput=CrossProduct(A=Transform(None=@n45) {transformSourceType=TRANSFORMSOURCE_View}, B=@n46))
  n48 = Add(A=Multiply(A=ComponentMask(None=@n44) {r=true}, B=@n47), B=Multiply(A=ComponentMask(None=@n44) {g=true}, B=CrossProduct(A=@n46, B=@n47)))
  n49 = Multiply(A=ComponentMask(None=@n44) {b=true}, B=@n46)
  n50 = ScalarParameter() {
   desc="LockedZ only for the specular highlight.\r\nfor a streched highlight in on the Z axis use the Blinn highlight"
   group=02_Anime Parameters
   parameterName=SpecularHighlight_LockedZVector
   sliderMax=1
  }
  n51 = OneMinus(None=@n50)
  n52 = Multiply(
   A=NamedRerouteDeclaration(None=Add(A=@n48, B=@n49)) {name=LightVector02, nodeColor=<rgba 0 1 0.7891 1>}
   B=AppendVector(A=Constant2Vector() {g=1, r=1}, B=Saturate(None=@n51))
  )
  n53 = Normalize(VectorInput=NamedRerouteUsage())
  n54 = Saturate(None=DotProduct(A=Normalize(VectorInput=@n52), B=ReflectionVectorWS(CustomWorldNormal=@n53)) {desc=Phong})
  NamedRerouteDeclaration(None=Multiply(A=Power(Base=@n54) {constExponent=3}, B=ScalarParameter() {defaultValue=1, group=010_Superlevel, parameterName=Opacity})) {name=Subsurface, nodeColor=<rgba 0.125 1 0 1>}
  NamedRerouteDeclaration(None=@n26.UVs) {name=CylinderProceduralUV, nodeColor=<rgba 1 0.9375 0 1>}
  NamedRerouteDeclaration(None=@n26.DDX) {name=DDX_Cylinder, nodeColor=<rgba 1 0 0.4219 1>}
  NamedRerouteDeclaration(None=@n26.DDY) {name=DDY_Cylinder, nodeColor=<rgba 0 0.7422 1 1>}
  n55 = LinearInterpolate(B=NamedRerouteUsage(), Alpha=NamedRerouteUsage()) {constA=1}
  n56 = Saturate(None=@n2)
  n57 = OneMinus(None=NamedRerouteUsage())
  n58 = NamedRerouteUsage()
  n59 = StaticSwitchParameter(
   True=Multiply(A=LinearInterpolate(A=@n55, Alpha=@n56) {constA=1}, B=@n57)
   False=StaticSwitchParameter(True=LinearInterpolate(Alpha=@n58) {constA=1, constB=0}, False=@n57) {group=010_Superlevel_Control, parameterName=UseEye}
  ) {defaultValue=true, group=010_Superlevel_Control, parameterName=UseNuclearMask}
  n60 = Saturate(None=Add(A=@n59, B=NamedRerouteUsage()))
  n61 = Subtract(A=NamedRerouteUsage(), B=NamedRerouteUsage())
  NamedRerouteDeclaration(None=LinearInterpolate(A=StaticSwitchParameter(True=@n60, False=@n59) {group=010_Superlevel_Control, parameterName=UseEye}, Alpha=Saturate(None=@n61)) {constB=0}) {name=NuclearMask, nodeColor=<rgba 1 0 0.1172 1>}
  n62 = StaticSwitchParameter(True=Transform(None=NamedRerouteUsage()), False=NamedRerouteUsage()) {defaultValue=true, group=010_Superlevel_Control, parameterName=UseNuclearMask}
  n63 = Constant3Vector() {constant=<rgba 0 0 1 0>}
  n64 = Normalize(VectorInput=Multiply(A=Transform(None=@n63), B=VectorParameter() {group=010_Superlevel, parameterName=LockDirection, defaultValue=<rgba 1 1 0.1 1>}.RGB))
  n65 = NamedRerouteUsage()
  n66 = Multiply(
   A=ScalarParameter() {defaultValue=0.5, group=010_Superlevel_Control, parameterName=LockZNormalBlend}
   B=OneMinus(None=Fresnel(Normal=@n65) {baseReflectFraction=0, exponent=1})
  )
  NamedRerouteDeclaration(None=Normalize(VectorInput=LinearInterpolate(A=@n62, B=@n64, Alpha=@n66))) {name=EdgeNormalWS, nodeColor=<rgba 0.8047 1 0 1>}
  NamedRerouteDeclaration(None=SmoothStep(Value=@n41) {constMax=0.6, constMin=0.3}) {name=SideMask, nodeColor=<rgba 0.6172 1 0 1>}
  n67 = VectorParameter() {group=010_Superlevel, parameterName=BackGroundColor, defaultValue=<rgba 0.1647 0.4948 0.0003 1>}
  n68 = VectorParameter() {group=010_Superlevel, parameterName=BackGroundColor_Tinted, defaultValue=<rgba 0.7832 0.8628 0 1>}
  NamedRerouteDeclaration(None=LinearInterpolate(A=Multiply(A=@n67.RGB, B=@n67.RGB), B=Multiply(A=@n68.RGB, B=@n68.RGB), Alpha=NamedRerouteUsage())) {name=BackGroundColor, nodeColor=<rgba 0 1 0.9766 1>}
  n69 = MaterialFunctionCall!(In=MaterialFunctionCall() {materialFunction=/Game/Global/Materials/Character_Parent_2023/Anime_Parent_2023/MaterialFunctions/MFC_AnimeShader_LightDirection}.Direction Vector) {materialFunction=/Game/Packages/Fortress_SharedMaterials/Cinematics/Anime/MF_CINE_AnimeLightDirOverride}
  n70 = MaterialFunctionCall(Float3=@n69.Result) {materialFunction=&BreakOutFloat3Components}
  n71 = ScalarParameter() {
   desc="0.0 = the light vector is locked on the XY plane.\r\n1.0 = the actual light vector is used  \r\n(for Faces use a blend value of 0,0)"
   group=02_Anime Parameters
   parameterName=LockedZLightVector_Blend
   sliderMax=1
  }
  n72 = MaterialFunctionCall(X=@n70.R, Y=@n70.R, Z=Multiply(A=@n70.B, B=@n71)) {materialFunction=&MakeFloat3}
  n73 = StaticSwitchParameter(True=LinearInterpolate(A=@n69.Result, B=Normalize(VectorInput=@n72.Result), Alpha=VertexColor().A), False=@n69.Result) {
   desc="Lock the Z axis of the Light Vector\r\nis driven by the VertexColor A channel (0 = Regular LightVector, 1= Locked LIghtVector)"
   group=02_Anime Parameters
   parameterName=UseLockedZLightVector
   sortPriority=0
  }
  n74 = StaticSwitchParameter(True=MaterialFunctionCall() {materialFunction=&CameraDirectionVector}.Result, False=@n73) {
   desc="Use the Camera vector instead of the light vector which is defined by the FX map. Select the FX channel with this parameter \"CameraLightVector_FXChannel\"."
   group=02_Anime Parameters
   parameterName=UseCameraLightVector
   sortPriority=0
  }
  NamedRerouteDeclaration(None=@n74) {name=OG_LightVector, nodeColor=<rgba 0.9922 0 1 1>}
  n75 = LinearInterpolate(
   A=MaterialFunctionCall(Normal=@n12, Flatness=Constant() {r=1}) {materialFunction=&FlattenNormal}.Result
   B=@n12
   Alpha=NamedRerouteUsage()
  )
  n76 = Constant3Vector() {constant=<rgba -1 -1 1 0>}
  n77 = LinearInterpolate(
   A=LinearInterpolate(A=@n75, B=Multiply(A=@n75, B=@n76), Alpha=NamedRerouteUsage())
   B=Constant3Vector() {constant=<rgba 0 0 1 0>}
   Alpha=OneMinus(None=NamedRerouteUsage())
  )
  n78 = Constant3Vector() {constant=<rgba 0 0 1 0>}
  NamedRerouteDeclaration(None=MaterialFunctionCall(
    BaseNormal=NamedRerouteUsage()
    AdditionalNormal=StaticSwitchParameter(True=@n77, False=@n78) {defaultValue=true, group=010_Superlevel_Control, parameterName=UseNuclearMask}
   ) {materialFunction=&BlendAngleCorrectedNormals}.Result) {name=EdgeNormal, nodeColor=<rgba 0.9922 0 1 1>}
  n79 = ScalarParameter() {defaultValue=0.2, group=010_Superlevel, parameterName=MinBackgroundColor}
  n80 = ScalarParameter() {defaultValue=0.5, group=010_Superlevel, parameterName=MinBackgroundColor_Tinted}
  n81 = NamedRerouteUsage()
  n82 = Constant3Vector() {constant=<rgba 0 0 1 0>}
  n83 = NamedRerouteUsage()
  n84 = LinearInterpolate(
   A=Multiply(A=NamedRerouteUsage(), B=LinearInterpolate(A=@n79, B=@n80, Alpha=@n81))
   B=NamedRerouteUsage()
   Alpha=OneMinus(None=MaterialFunctionCall(Axis=@n82, Normal=@n83) {materialFunction=&AxisAlignedFresnel}.Fresnel)
  ) {constA=0.2}
  n85 = NamedRerouteDeclaration(None=@n84) {name=BackGroundColor_OUT, nodeColor=<rgba 0.6641 1 0 1>}
  n86 = Normalize(VectorInput=NamedRerouteUsage())
  n87 = Saturate(None=DotProduct(A=ReflectionVectorWS(CustomWorldNormal=@n86) {bNormalizeCustomWorldNormal=true}, B=CameraVectorWS()))
  n88 = Add(
   A=SmoothStep(Value=Power(Base=@n87) {constExponent=10}) {constMax=0.5, constMin=0.3}
   B=Multiply(A=Fresnel(Normal=@n86) {baseReflectFraction=0, exponent=2}) {constB=0.5}
  )
  n89 = NamedRerouteUsage()
  n90 = Add(
   A=Multiply(A=Multiply(A=@n88, B=@n89), B=NamedRerouteUsage())
   B=Multiply(A=Fresnel(Normal=@n86) {baseReflectFraction=0.5, exponent=2}, B=NamedRerouteUsage())
  )
  n91 = Multiply(A=OneMinus(None=@n84), B=NamedRerouteUsage())
  n92 = NamedRerouteUsage()
  NamedRerouteDeclaration(None=LinearInterpolate(A=LinearInterpolate(A=@n85, B=@n90, Alpha=@n91), B=NamedRerouteUsage(), Alpha=Multiply(A=@n91, B=@n92))) {name=BubbleFinal, nodeColor=<rgba 0 1 0.4844 1>}
  n93 = ScalarParameter() {defaultValue=1.5, group=010_Superlevel, parameterName=RimLight}
  n94 = ScalarParameter() {defaultValue=0.15, group=010_Superlevel, parameterName=RimLight_Tint}
  n95 = NamedRerouteUsage()
  NamedRerouteDeclaration(None=Multiply(A=Power(Base=@n54) {constExponent=20}, B=LinearInterpolate(A=@n93, B=@n94, Alpha=@n95))) {name=TransHigh, nodeColor=<rgba 0 1 0.4141 1>}
  n96 = Custom!(None) {
   code="// The below expression will get compiled\r\n// into the output of this node\r\nfloat3 T = Parameters.TangentToWorld[1];\r\n\r\nreturn T;"
   showCode=true
  }
  NamedRerouteDeclaration(None=Transform(None=@n96) {transformSourceType=TRANSFORMSOURCE_World, transformType=TRANSFORM_Tangent}) {name=T, nodeColor=<rgba 0 1 0.3906 1>}
  n97 = TextureSampleParameter2D() {
   group=010_Superlevel
   parameterName=FixMask
   samplerType=SAMPLERTYPE_Masks
   texture=/BRCosmetics/Characters/Player/Male/Medium/Bodies/M_MED_SilentNovel_Vane/Skins/Superlevel/Texture/T_SilentNovel_Vane_Body_Superlevel_Fix
  }
  n98 = Constant()
  NamedRerouteDeclaration(None=StaticSwitchParameter(True=@n97.R, False=@n98) {defaultValue=true, group=010_Superlevel_Fix, parameterName=UseNormalFix}) {name=NormalMask, nodeColor=<rgba 1 0.3984 0 1>}
  n99 = VectorParameter() {group=010_Superlevel, parameterName=BubbleSpeed, defaultValue=<rgba 0.01 -0.05 0.01 -0.5>}
  NamedRerouteDeclaration(None=ComponentMask(None=@n99.RGBA) {a=true, b=true}) {name=DrippingSpeed, nodeColor=<rgba 0 1 0.8359 1>}
  n100 = Normalize(VectorInput=VectorParameter() {group=010_Superlevel, parameterName=ScreenSpaceLightVector01, defaultValue=<rgba 1 1 0 1>}.RGB)
  n101 = Constant3Vector() {constant=<rgba 0 1 0 0>}
  n102 = Normalize(VectorInput=WorldPosition() {worldPositionShaderOffset=WPT_CameraRelative}.XYZ)
  n103 = Normalize(VectorInput=CrossProduct(A=Transform(None=@n101) {transformSourceType=TRANSFORMSOURCE_View}, B=@n102))
  n104 = Multiply(A=ComponentMask(None=@n100) {r=true}, B=@n103)
  n105 = Multiply(A=ComponentMask(None=@n100) {g=true}, B=CrossProduct(A=@n102, B=@n103))
  n106 = ComponentMask(None=@n100) {b=true}
  NamedRerouteDeclaration(None=Add(A=Add(A=@n104, B=@n105), B=Multiply(A=@n106, B=@n102))) {name=LightVector01, nodeColor=<rgba 0 1 0.7891 1>}
  n107 = Add(A=NamedRerouteUsage(), B=@n27.G)
  NamedRerouteDeclaration(None=StaticSwitchParameter(True=Saturate(None=@n107), False=@n27.G) {group=010_Superlevel_Fix, parameterName=UseEyeAsTinted}) {name=ColorTintMask, nodeColor=<rgba 0.0781 1 0 1>}
  n108 = VectorParameter() {group=010_Superlevel, parameterName=SaturatedTrasimission, defaultValue=<rgba 0.078 0.3031 0.0003 1>}
  n109 = VectorParameter() {group=010_Superlevel, parameterName=SaturatedTrasimission_Tinted, defaultValue=<rgba 0.698 1 0 1>}
  NamedRerouteDeclaration(None=LinearInterpolate(A=Multiply(A=@n108.RGB, B=@n108.RGB), B=Multiply(A=@n109.RGB, B=@n109.RGB), Alpha=NamedRerouteUsage())) {name=TransColor, nodeColor=<rgba 0.3125 0 1 1>}
  n110 = VectorParameter() {group=010_Superlevel, parameterName=BubbleHightLightColor, defaultValue=<rgba 0.0517 0.3031 0.0155 1>}
  n111 = VectorParameter() {group=010_Superlevel, parameterName=BubbleHightLightColor_Tinted, defaultValue=<rgba 0.4771 0.6115 0.1 1>}
  NamedRerouteDeclaration(None=LinearInterpolate(A=Multiply(A=@n110.RGB, B=@n110.RGB), B=Multiply(A=@n111.RGB, B=@n111.RGB), Alpha=NamedRerouteUsage())) {name=HighLightColor, nodeColor=<rgba 0 0.5312 1 1>}
  n112 = VectorParameter() {group=010_Superlevel, parameterName=BubbleColor, defaultValue=<rgba 0.388 0.9448 0.1781 1>}
  n113 = VectorParameter() {group=010_Superlevel, parameterName=BubbleColor_Tinted, defaultValue=<rgba 0.4621 0.6448 0.0329 2>}
  NamedRerouteDeclaration(None=LinearInterpolate(A=Multiply(A=@n112.RGB, B=@n112.RGB), B=Multiply(A=@n113.RGB, B=@n113.RGB), Alpha=NamedRerouteUsage())) {name=BubbleColor, nodeColor=<rgba 0.3125 0 1 1>}
  n114 = VectorParameter() {group=010_Superlevel, parameterName=BubbleOutlineColor, defaultValue=<rgba 0.0535 0.1323 0 1>}
  n115 = VectorParameter() {group=010_Superlevel, parameterName=BubbleOutline_Tint, defaultValue=<rgba 0.0953 0.1365 0.0096 1>}
  NamedRerouteDeclaration(None=LinearInterpolate(A=Multiply(A=@n114.RGB, B=@n114.RGB), B=Multiply(A=@n115.RGB, B=@n115.RGB), Alpha=NamedRerouteUsage())) {name=BubbleOutlineColor, nodeColor=<rgba 1 0.8438 0 1>}
  n116 = VertexColor()
  n117 = AppendVector(A=@n116, B=@n116)
  n118 = TextureSampleParameter2D() {
   group=01_EyeAnimation_Parameters
   parameterName=EyeMask
   samplerSource=SSM_Wrap_WorldGroupSettings
   samplerType=SAMPLERTYPE_Masks
   sortPriority=-32
   texture=&T_WhiteMask
  }
  n119 = Subtract(A=ComponentMask(None=TextureCoordinate() {coordinateIndex=1}) {r=true}) {constB=0.0001}
  n120 = StaticSwitchParameter(
   True=ChannelMaskParameter(None=StaticSwitchParameter(True=@n117, False=@n118.RGBA) {group=01_EyeAnimation_Parameters, parameterName=UseVertexColorForEyeMask, sortPriority=-33}) {group=01_EyeAnimation_Parameters, parameterName=EyeMaskChannel, sortPriority=-32}.R
   False=Saturate(None=Sign(None=@n119))
  ) {group=01_EyeAnimation_Parameters, parameterName=UseEyeMask, sortPriority=-33}
  NamedRerouteDeclaration(None=@n120) {name=EyeMask, nodeColor=<rgba 0.1953 1 0 1>}
  n121 = ComponentMask(None=@n99.RGB) {g=true, r=true}
  n122 = TextureObjectParameter() {
   group=010_Superlevel
   parameterName=BubblePattern
   texture=&T_Superlevel_Bubbles
  }
  n123 = TextureSample(UVs=Panner(Coordinate=NamedRerouteUsage(), Speed=@n121) {speedY=0.1}, Tex=@n122) {
   samplerSource=SSM_Wrap_WorldGroupSettings
   texture=&T_Superlevel_Bubbles
  }
  n124 = TextureSample(UVs=Panner(Coordinate=NamedRerouteUsage(), Speed=@n121) {speedY=0.1}, Tex=@n122) {
   samplerSource=SSM_Wrap_WorldGroupSettings
   texture=&T_Superlevel_Bubbles
  }
  n125 = NamedRerouteUsage()
  n126 = LinearInterpolate(A=@n123.RGB, B=@n124.RGB, Alpha=@n125)
  n127 = DeriveNormalZ(InXY=Subtract(A=ComponentMask(None=@n126) {g=true, r=true}, B=Constant() {r=0.5}))
  n128 = Multiply(A=@n127, B=Constant3Vector() {constant=<rgba -1 -1 1 0>})
  n129 = NamedRerouteUsage()
  NamedRerouteDeclaration(None=Transform(None=LinearInterpolate(A=@n127, B=@n128, Alpha=@n129))) {name=BubbleNormalWS, nodeColor=<rgba 0 0.5078 1 1>}
  n130 = NamedRerouteUsage()
  n131 = Subtract(A=NamedRerouteUsage(), B=NamedRerouteUsage())
  n132 = LinearInterpolate(
   A=@n130
   B=Multiply(A=NamedRerouteUsage(), B=VectorParameter() {group=010_Superlevel_Eye, parameterName=EyeTint, defaultValue=<rgba 0.7017 1 0.6135 1>}.RGB)
   Alpha=NamedRerouteDeclaration(None=Saturate(None=@n131)) {name=EyeWithoutLids, nodeColor=<rgba 0 0.4141 1 1>}
  )
  n133 = Multiply(
   A=NamedRerouteUsage()
   B=VectorParameter() {group=010_Superlevel, parameterName=AdditionalEmissiveColor, defaultValue=<rgba 0.0006 0.924 0.0203 1>}.RGB
  )
  NamedRerouteDeclaration(None=Add(A=StaticSwitchParameter(True=@n132, False=@n130) {group=010_Superlevel_Control, parameterName=UseEye}, B=@n133)) {name=LitOUT, nodeColor=<rgba 0 1 0.6016 1>}
  n134 = Saturate(None=@n130)
  n135 = VectorParameter() {group=010_Superlevel, parameterName=BackLightTint, defaultValue=<rgba 0.4659 0.9073 0 1>}
  n136 = VectorParameter() {group=010_Superlevel, parameterName=BackLightTint_Tinted, defaultValue=<rgba 0.7591 0.8323 0 1>}
  n137 = LinearInterpolate(
   A=Multiply(A=@n134, B=ScalarParameter() {defaultValue=0.5, group=010_Superlevel, parameterName=ColorShadedIntensity})
   B=LinearInterpolate(A=Multiply(A=@n134, B=@n135.RGB), B=Multiply(A=@n134, B=@n136.RGB), Alpha=NamedRerouteUsage())
   Alpha=NamedRerouteUsage()
  )
  n138 = LinearInterpolate(
   A=@n137
   B=Multiply(
    A=NamedRerouteUsage()
    B=VectorParameter() {group=010_Superlevel_Eye, parameterName=EyeTint_Shaded, defaultValue=<rgba 0.7017 1 0.6135 1>}.RGB
   )
   Alpha=NamedRerouteUsage()
  )
  NamedRerouteDeclaration(None=Add(A=StaticSwitchParameter(True=@n138, False=@n137) {group=010_Superlevel_Control, parameterName=UseEye}, B=@n133)) {name=ShadedOUT, nodeColor=<rgba 0.5 0 1 1>}
  NamedRerouteDeclaration(None=@n29) {name=UV_RG, nodeColor=<rgba 0.1719 0 1 1>}
  n139 = Multiply(A=NamedRerouteUsage()) {constB=5}
  n140 = NamedRerouteUsage()
  n141 = OneMinus(None=Saturate(None=Add(A=@n139, B=@n140)))
  NamedRerouteDeclaration(None=Multiply(A=@n141, B=LinearInterpolate(A=@n123.B, B=@n124.B, Alpha=@n125))) {name=BubbleMask, nodeColor=<rgba 0.1719 1 0 1>}
  NamedRerouteDeclaration(None=@n27.B) {name=GradientMask, nodeColor=<rgba 0.6406 0 1 1>}
  n142 = VectorParameter() {group=010_Superlevel, parameterName=GreenUpper, defaultValue=<rgba 0.073 0.274 0 1>}
  n143 = Multiply(A=@n142.RGB, B=@n142.RGB)
  n144 = VectorParameter() {group=010_Superlevel, parameterName=PinkLower, defaultValue=<rgba 0.8281 0 0.7167 1>}
  n145 = Multiply(A=@n144.RGB, B=@n144.RGB)
  n146 = SmoothStep(Value=Power(Base=NamedRerouteUsage(), Exp=ScalarParameter() {defaultValue=3, group=010_Superlevel, parameterName=PinkPower}) {constExponent=3}) {constMax=0.8}
  n147 = ComponentMask(None=NamedRerouteUsage()) {b=true}
  n148 = ScalarParameter() {defaultValue=200, parameterName=DesaturationUnit}
  n149 = Power(Base=Saturate(None=Divide(A=@n147, B=@n148)))
  NamedRerouteDeclaration(None=Desaturation(None=LinearInterpolate(A=@n143, B=@n145, Alpha=@n146), Fraction=@n149)) {name=LineGradientColor, nodeColor=<rgba 0 1 0.3203 1>}
  n150 = Subtract(
   A=NamedRerouteUsage()
   B=ComponentMask(None=VectorParameter() {group=010_Superlevel, parameterName=UV_Offset, defaultValue=<rgba -0.2 0 0 1>}.RGB) {g=true, r=true}
  )
  n151 = ScalarParameter() {defaultValue=110, group=010_Superlevel, parameterName=DistortionUnit}
  n152 = VectorParameter() {group=010_Superlevel, parameterName=PatternSpeed, defaultValue=<rgba 0 -0.05 0 -0.02>}
  n153 = TextureSample(UVs=Panner(Coordinate=Divide(A=@n150, B=@n151), Speed=AppendVector(A=@n152.R, B=@n152.R))) {samplerType=SAMPLERTYPE_LinearColor, texture=/Game/Global/Textures/Default/Flowmaps/WindFlowMap}
  n154 = Multiply(
   A=ComponentMask(None=Subtract(A=@n153.RGB) {constB=0.5}) {g=true, r=true}
   B=ScalarParameter() {defaultValue=0.2, group=010_Superlevel, parameterName=DistortionAmount}
  )
  n155 = Panner(
   Coordinate=Divide(A=@n150, B=ScalarParameter() {defaultValue=90, group=010_Superlevel, parameterName=PatternUnit})
   Speed=AppendVector(A=@n152.B, B=@n152.B)
  )
  n156 = NamedRerouteUsage()
  n157 = ScalarParameter() {defaultValue=0.5, group=010_Superlevel, parameterName=SideDistortIntensity}
  n158 = TextureSampleParameter2D(UVs=Add(A=Add(A=@n154, B=@n155), B=Multiply(A=@n156, B=@n157) {constB=0.1})) {
   group=010_Superlevel
   parameterName=SDF_Pattern
   samplerType=SAMPLERTYPE_Masks
   texture=/BRCosmetics/Characters/Player/Male/Medium/Bodies/M_MED_SilentNovel_Vane/Skins/Superlevel/Texture/T_Superlevel_Patttern
  }
  NamedRerouteDeclaration(None=@n158.R) {name=LineMask, nodeColor=<rgba 1 0 0 1>}
  n159 = VectorParameter() {group=010_Superlevel, parameterName=PurplePatternColor_Dark, defaultValue=<rgba 0.0285 0 0.0698 1>}
  n160 = Multiply(A=@n159.RGB, B=@n159.RGB)
  n161 = VectorParameter() {group=010_Superlevel, parameterName=PurplePatternColor_Light, defaultValue=<rgba 0.294 0.0238 0.4427 1>}
  n162 = Multiply(A=@n161.RGB, B=@n161.RGB)
  n163 = SmoothStep(Value=Power(Base=NamedRerouteUsage()) {constExponent=3}) {constMax=0.8}
  NamedRerouteDeclaration(None=Desaturation(None=LinearInterpolate(A=@n160, B=@n162, Alpha=@n163), Fraction=@n149)) {name=PurpleColor, nodeColor=<rgba 0 0.3672 1 1>}
  n164 = NamedRerouteUsage()
  NamedRerouteDeclaration(None=SphereMask(A=ComponentMask(None=@n164) {g=true}, B=Constant() {r=0.5}) {attenuationRadius=0.45, hardnessPercent=0.5}) {name=G01_NormalMask, nodeColor=<rgba 1 0 0.1641 1>}
  NamedRerouteDeclaration(None=StaticSwitchParameter(True=@n97.G, False=Constant()) {defaultValue=true, group=010_Superlevel_Fix, parameterName=UseAreaFix}) {name=AreaMask, nodeColor=<rgba 0.4531 0 1 1>}
  n165 = Constant3Vector() {constant=<rgba 0 0 -1 0>}
  n166 = NamedRerouteUsage()
  n167 = Saturate(None=DotProduct(A=NamedRerouteDeclaration(None=@n165) {name=LightVector03, nodeColor=<rgba 0 1 0.7891 1>}, B=Normalize(VectorInput=@n166)) {desc=Phong})
  NamedRerouteDeclaration(None=Multiply(A=Power(Base=@n167), B=NamedRerouteUsage())) {name=BottomColor, nodeColor=<rgba 0.7578 0 1 1>}
  n168 = VectorParameter() {group=010_Superlevel, parameterName=BottomLitColor, defaultValue=<rgba 0.0406 0.349 0 1>}
  n169 = VectorParameter() {group=010_Superlevel, parameterName=BottomLitColor_Tinted, defaultValue=<rgba 0.1941 0.2448 0 1>}
  NamedRerouteDeclaration(None=LinearInterpolate(A=Multiply(A=@n168.RGB, B=@n168.RGB), B=Multiply(A=@n169.RGB, B=@n169.RGB), Alpha=NamedRerouteUsage())) {name=BottonTint, nodeColor=<rgba 0.2656 0 1 1>}
  NamedRerouteDeclaration(None=@n43) {name=SideDistortion, nodeColor=<rgba 0 1 0.1328 1>}
  NamedRerouteDeclaration(None=Multiply(A=@n141, B=LinearInterpolate(A=@n123.A, B=@n124.A, Alpha=@n125))) {name=BubbleLine, nodeColor=<rgba 1 0.3047 0 1>}
  n170 = FunctionInput(Preview=Constant3Vector() {constant=<rgba 1 1 1 0>}) {bUsePreviewValueAsDefault=true, inputName=Color_Lit, sortPriority=1}
  n171 = FunctionInput(Preview=Constant3Vector() {constant=<rgba 0.6031 0.6031 0.6031 0>}) {bUsePreviewValueAsDefault=true, inputName=Color_Shaded, sortPriority=2}
  n172 = StaticBoolParameter() {defaultValue=true, group=01_EyeAnimation_Parameters, parameterName=Use UV2 for Eye Animation, sortPriority=-30}
  n173 = MaterialFunctionCall!(Color Lit=@n170, Color Shaded=@n171, Use UV2=@n172, Mask) {materialFunction=/BRCosmetics/Characters/Player/Male/Medium/Bodies/M_MED_GrimeHold/Skins/Superlevel/Material/MF_AnimeEyeAnimation_3eyes_SL_v2}
  n174 = Add(A=@n173.UpperLidMask, B=@n173.UpperLidMask)
  n175 = MaterialFunctionCall!(Color Lit=@n170, Color Shaded=@n171, Use UV2=@n172, Mask=@n120) {materialFunction=&/MF_AnimeEyeAnimation_Porridge_DF_SL}
  n176 = Add(A=@n175.UpperLidMask, B=@n175.UpperLidMask)
  n177 = NamedRerouteDeclaration(None=@n175.DFMask) {name=DF_Lids, nodeColor=<rgba 0.8047 1 0 1>}
  NamedRerouteDeclaration(None=StaticSwitchParameter(True=Multiply(A=@n174, B=@n173.DF_Lids), False=Multiply(A=@n176, B=@n177)) {group=01_EyeAnimation_Parameters, parameterName=Use3eyes, sortPriority=-33}) {name=Lidmask, nodeColor=<rgba 0 1 0.9531 1>}
  n178 = ScalarParameter() {defaultValue=0.4, group=010_Superlevel, parameterName=MaskHigh}
  n179 = Subtract(A=@n178) {constB=0.05}
  n180 = NamedRerouteUsage()
  n181 = NamedRerouteDeclaration(None=SmoothStep(Min=@n179, Max=@n178, Value=@n180) {constMax=0.4, constMin=0.3}) {name=NuclearMaskDF, nodeColor=<rgba 1 0.2109 0 1>}
  n182 = LinearInterpolate(B=Multiply(A=OneMinus(None=@n181), B=SphereMask(A=@n180, B=@n179) {attenuationRadius=0.1, hardnessPercent=1}), Alpha=NamedRerouteUsage()) {constB=0}
  n183 = NamedRerouteUsage()
  NamedRerouteDeclaration(None=StaticSwitchParameter(True=LinearInterpolate(A=@n182, Alpha=@n183) {constB=0}, False=Constant()) {defaultValue=true, group=010_Superlevel_Control, parameterName=UseNuclearMask}) {name=FakeAO, nodeColor=<rgba 0.4531 0 1 1>}
  NamedRerouteDeclaration(None=@n28.PS) {name=PreskinnedPOS, nodeColor=<rgba 1 0 0.5391 1>}
  n184 = Abs(None=ComponentMask(None=NamedRerouteUsage()) {r=true})
  n185 = ScalarParameter() {defaultValue=10, group=010_Superlevel, parameterName=MaskDist}
  n186 = SphereMask(A=Saturate(None=Divide(A=@n184, B=@n185)), B=Constant()) {attenuationRadius=1, hardnessPercent=1}
  NamedRerouteDeclaration(None=StaticSwitchParameter(True=OneMinus(None=@n186), False=Constant() {r=1}) {defaultValue=true, group=010_Superlevel_Control, parameterName=UseCenterMask}) {name=CenterMask, nodeColor=<rgba 0 0.8594 1 1>}
  n187 = NamedRerouteDeclaration(None=@n11.B) {name=NoiseGradient, nodeColor=<rgba 1 0 0.8438 1>}
  n188 = Multiply(A=NamedRerouteUsage(), B=NamedRerouteUsage())
  n189 = Add(A=Add(A=@n187, B=@n188), B=NamedRerouteUsage())
  NamedRerouteDeclaration(None=Multiply(A=Saturate(None=@n189), B=NamedRerouteUsage())) {name=DF_Sum, nodeColor=<rgba 1 0.7266 0 1>}
  n190 = NamedRerouteUsage()
  n191 = ScalarParameter() {defaultValue=0.5, group=010_Superlevel, parameterName=DippingIntensity}
  n192 = NamedRerouteUsage()
  n193 = NamedRerouteUsage()
  n194 = Multiply(A=Multiply(A=Multiply(A=@n190, B=@n191) {constB=2}, B=@n192), B=@n193)
  n195 = NamedRerouteUsage()
  n196 = VectorParameter() {group=010_Superlevel, parameterName=CenterSaturation, defaultValue=<rgba 0 0 120 1>}
  n197 = Length(None=ComponentMask(None=Subtract(A=@n195, B=@n196.RGB)) {b=true, r=true}) {is=class MaterialExpressionLength}
  n198 = ScalarParameter() {defaultValue=30, group=010_Superlevel, parameterName=SaturationRadius}
  n199 = VectorParameter() {group=010_Superlevel, parameterName=CenterSaturation2, defaultValue=<rgba 0 0 0 1>}
  n200 = Length(None=ComponentMask(None=Subtract(A=@n195, B=@n199.RGB)) {b=true, r=true})
  n201 = ScalarParameter() {defaultValue=30, group=010_Superlevel, parameterName=SaturationRadius1}
  n202 = Min(A=Saturate(None=Divide(A=@n197, B=@n198)), B=Saturate(None=Divide(A=@n200, B=@n201)))
  n203 = Multiply(A=@n192, B=SmoothStep(Value=Power(Base=@n202) {constExponent=2.2}) {constMin=0.3})
  n204 = Multiply(A=@n193, B=@n203)
  NamedRerouteDeclaration(None=StaticSwitchParameter(True=Add(A=@n194, B=@n204), False=@n203) {defaultValue=true, group=010_Superlevel_Control, parameterName=UseNuclearMask}) {name=FakeSaturation, nodeColor=<rgba 0.4531 0 1 1>}
  n205 = OneMinus(None=SmoothStep(Max=@n178, Value=@n187))
  n206 = Multiply(A=Multiply(A=@n205, B=SmoothStep(Min=@n179, Value=@n188) {constMax=0.8})) {constB=0.5}
  n207 = Multiply(A=Multiply(A=@n188, B=NamedRerouteUsage()), B=@n205)
  NamedRerouteDeclaration(None=LinearInterpolate(A=Add(A=@n206, B=@n207), Alpha=NamedRerouteUsage()) {constB=0}) {name=DripMask, nodeColor=<rgba 0.4531 0 1 1>}
  n208 = Add(A=@n27.R, B=@n27.R)
  NamedRerouteDeclaration(None=OneMinus(None=@n208)) {name=FullyGradientMask, nodeColor=<rgba 0.8516 0 1 1>}
  n209 = FunctionInput(Preview=Constant3Vector() {constant=<rgba 0 0 1 0>}) {bUsePreviewValueAsDefault=true, inputName=Normal_WS}
  NamedRerouteDeclaration(None=Transform(None=NamedRerouteDeclaration(None=@n209) {name=NormalMap_IN_WS, nodeColor=<rgba 1 0.4453 0 1>}) {transformSourceType=TRANSFORMSOURCE_World, transformType=TRANSFORM_Tangent}) {name=NormalMap_IN, nodeColor=<rgba 1 0.75 0 1>}
  MaterialExpressionMultiply_56 = Multiply(A=@n7, B=ScalarParameter() {defaultValue=1, group=010_Superlevel, parameterName=DrippingDpubleTiling}) {constB=2}
  n210 = Convert(None=@MaterialExpressionMultiply_56) {
   nodeName=Break Float2
   convertInputs [{
     type=Vector2
     expressionInput {
      expression=&self:MaterialExpressionMultiply_56
      inputName=None
      mask=0
      maskA=0
      maskB=0
      maskG=0
      maskR=0
      outputIndex=0
     }
    }]
   convertOutputs [{type=Scalar}, {type=Scalar}]
  }
  n211 = ScalarParameter() {defaultValue=0.3, group=010_Superlevel, parameterName=PowerUVG}
  n212 = LinearInterpolate(
   A=@MaterialExpressionMultiply_56
   B=AppendVector(A=@n210, B=Power(Base=@n210, Exp=@n211) {constExponent=5})
   Alpha=ScalarParameter() {defaultValue=0.65, parameterName=NolinearGAlpha}
  )
  n213 = TextureObjectParameter() {
   group=010_Superlevel
   parameterName=DrippingPattern
   samplerType=SAMPLERTYPE_LinearColor
   texture=/BRCosmetics/Characters/Player/Male/Medium/Bodies/M_MED_SilentNovel_Vane/Skins/Superlevel/Texture/T_Superlevel_DrippingMask02
  }
  n214 = TextureSample!(
   UVs=Panner(Coordinate=NamedRerouteDeclaration(None=@n212) {name=DrippingUV, nodeColor=<rgba 0.7344 1 0 1>}, Speed=NamedRerouteUsage())
   Tex=@n213
   "DDX(UVs)"=@n9
   "DDY(UVs)"=@n10
   Apply View MipBias
  ) {
   mipValueMode=TMVM_Derivative
   samplerSource=SSM_Wrap_WorldGroupSettings
   samplerType=SAMPLERTYPE_LinearColor
   texture=&T_Superlevel_Bubbles
  }
  n215 = NamedRerouteUsage()
  n216 = VectorParameter() {group=010_Superlevel, parameterName=BubbleSpeed02, defaultValue=<rgba 0 -0.05 0.01 -0.5>}
  n217 = TextureSample(
   UVs=Panner(Coordinate=@n215, Speed=ComponentMask(None=@n216.RGB) {g=true, r=true})
   Tex=@n213
   "DDX(UVs)"=NamedRerouteUsage()
   "DDY(UVs)"=NamedRerouteUsage()
  ) {
   mipValueMode=TMVM_Derivative
   samplerSource=SSM_Wrap_WorldGroupSettings
   samplerType=SAMPLERTYPE_LinearColor
   texture=&T_Superlevel_Bubbles
  }
  n218 = TextureSample(
   UVs=Panner(Coordinate=@n215, Speed=ComponentMask(None=@n216.RGBA) {a=true, b=true})
   Tex=@n213
   "DDX(UVs)"=NamedRerouteUsage()
   "DDY(UVs)"=NamedRerouteUsage()
  ) {
   mipValueMode=TMVM_Derivative
   samplerSource=SSM_Wrap_WorldGroupSettings
   samplerType=SAMPLERTYPE_LinearColor
   texture=&T_Superlevel_Bubbles
  }
  NamedRerouteDeclaration(None=Add(A=Add(A=@n214.R, B=@n217.G), B=@n218.B)) {name=DripSum, nodeColor=<rgba 1 0.375 0 1>}
  NamedRerouteDeclaration(None=StaticSwitchParameter(True=@n97.B, False=Constant()) {defaultValue=true, group=010_Superlevel_Fix, parameterName=UseArtifactFix}) {name=ArtifactFix, nodeColor=<rgba 0.2656 1 0 1>}
  NamedRerouteDeclaration(None=OneMinus(None=Add(A=@n208, B=@n27.B))) {name=BackArea, nodeColor=<rgba 1 0 0.2344 1>}
  n219 = VectorParameter() {group=010_Superlevel, parameterName=BackGroundColor_OutRadius, defaultValue=<rgba 0.3431 0.4948 0 1>}
  n220 = VectorParameter() {group=010_Superlevel, parameterName=BackGroundColor_OutRadius_Tinted, defaultValue=<rgba 0.7832 0.8628 0 1>}
  NamedRerouteDeclaration(None=LinearInterpolate(A=Multiply(A=@n219.RGB, B=@n219.RGB), B=Multiply(A=@n220.RGB, B=@n220.RGB), Alpha=NamedRerouteUsage())) {name=BackGroundColor_OutRadius, nodeColor=<rgba 0 1 0.9766 1>}
  NamedRerouteDeclaration(None=Add(A=Divide(A=@n36, B=@n30), B=@n43)) {name=ProceduralUV_Side, nodeColor=<rgba 1 0.9375 0 1>}
  n221 = DotProduct(A=@n39, B=Constant3Vector() {constant=<rgba 0 0 1 0>})
  NamedRerouteDeclaration(None=SmoothStep(Value=Saturate(None=@n221)) {constMax=0.9, constMin=0.5}) {name=TopDownMask, nodeColor=<rgba 0.0312 0 1 1>}
  NamedRerouteDeclaration(None=StaticSwitchParameter(True=@n97.R, False=@n98) {group=010_Superlevel_Fix, parameterName=UseRedAsEmissive}) {name=AdditionalEmissive, nodeColor=<rgba 0.4062 0 1 1>}
  n222 = TextureSampleParameter2D() {group=010_Superlevel, parameterName=DisappearMask, samplerType=SAMPLERTYPE_Masks, texture=&T_BlackMask}
  NamedRerouteDeclaration(None=StaticSwitchParameter(True=OneMinus(None=@n222.G), False=Constant() {r=1}) {group=010_Superlevel_Fix, parameterName=UseDisappearMask}) {name=DisappearMask, nodeColor=<rgba 0.6641 0 1 1>}
  NamedRerouteDeclaration(None=StaticSwitchParameter(True=@n173.Lit, False=@n175.Lit) {group=01_EyeAnimation_Parameters, parameterName=Use3eyes, sortPriority=-33}) {name=EyeLit, nodeColor=<rgba 0.3828 1 0 1>}
  NamedRerouteDeclaration(None=StaticSwitchParameter(True=@n173.Shaded, False=@n175.Shaded) {group=01_EyeAnimation_Parameters, parameterName=Use3eyes, sortPriority=-33}) {name=EyeShaded, nodeColor=<rgba 1 0 0.7031 1>}
  n223 = ComponentMask(None=@n164) {g=true}
  n224 = Constant() {r=1}
  n225 = Constant()
  n226 = Constant()
  NamedRerouteDeclaration(None=Multiply(A=If(A=@n223, "A > B"=@n224, "A == B"=@n225, "A < B"=@n225), B=If(A=@n223, "A > B"=@n226, "A == B"=@n226, "A < B"=@n224) {constB=1})) {name=G01Mask, nodeColor=<rgba 0 1 0.3906 1>}
  n227 = VertexColor()
  n228 = Subtract(A=Multiply(A=@n227) {constB=2})
  n229 = ScalarParameter() {desc="0.0 = surface normals, 1.0 baked vertex color normal", group=02_Anime Parameters, parameterName=VertexBakedNormal_Blend, sliderMax=1}
  Normalize(VectorInput=LinearInterpolate(A=@n209, B=Transform(None=@n228) {transformSourceType=TRANSFORMSOURCE_Local}, Alpha=Multiply(A=@n229, B=@n227.A)))
  TemporalResponsivenessOutput!(None=LinearInterpolate(
    A=ScalarParameter() {parameterName=NonEyesTRMulti}
    B=ScalarParameter() {defaultValue=1, parameterName=EyesTRMulti}
    Alpha=StaticSwitchParameter(True=@n173.EyeMask, False=@n175.EyeMask) {group=01_EyeAnimation_Parameters, parameterName=Use3eyes, sortPriority=-33}
   ))
 ]}
