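# Anime-style character material for the SilentNovel Vane Tier01 superlevel skin, using the Cloth shading model.
# Shared anime Pre, Calculations and Post functions form the pipeline, with MF_Superlevel_Sim after Pre and an eyeball mask on DF lines before Post.
# Static switches choose the normal fed to the sim: vertex-color baked normals, or height-faded spherical/cylindrical or baked capsule normals.
# The sphere normal and height fade use pre-skinned local position through vertex interpolators.
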Material /BRCosmetics/Characters/Player/Male/Medium/Bodies/M_MED_SilentNovel_Vane/Skins/Superlevel/Material/M_AnimShader_SilentNovel_Tier01
graph {
 properties {
  bAutomaticallySetUsageInEditor=false
  bUseMaterialAttributes=true
  bUsedWithMeshDeformer=true
  bUsedWithMorphTargets=true
  bUsedWithSkeletalMesh=true
  bUsedWithStaticLighting=true
  pixelDepthOffsetMode=PDOM_Legacy
  shadingModel=MSM_Cloth
  usedShadingModels=Cloth
 }
 nodes [
  n1 = MaterialFunctionCall!(CustomUV) {materialFunction=/Game/Global/Materials/Character_Parent_2023/Anime_Parent_2023/MaterialFunctions/MFC_AnimeShader_2023_Pre}
  n2 = Transform(None=@n1.Normal)
  n3 = TextureSampleParameter2D() {
   group=Toon_Shading
   parameterName=BakedCapsuleTangentNormal
   samplerType=SAMPLERTYPE_Normal
   texture=&FlatNormalBlue
  }
  n4 = VertexColor()
  n5 = LocalPosition() {localOrigin=InstancePreSkinning}
  n6 = SmoothStep(
   Min=ScalarParameter() {defaultValue=140, group=Toon_Shading, parameterName=NormalSphereFadeStart}
   Max=ScalarParameter() {defaultValue=150, group=Toon_Shading, parameterName=NormalSphereFadeEnd}
   Value=VertexInterpolator(VS=@n5.Z).PS
  ) {constMax=60, constMin=45}
  n7 = StaticSwitchParameter(True=Multiply(A=NamedRerouteDeclaration(None=@n4.A) {nodeColor=<rgba 0.7344 1 0 1>}, B=@n6), False=@n6) {group=Toon_Shading, parameterName=UseMultiplyVertAlphaForFade}
  n8 = LinearInterpolate(A=@n2, B=Transform(None=@n3.RGB), Alpha=@n7)
  n9 = Subtract(A=@n5.XYZ, B=VectorParameter() {group=Toon_Shading, parameterName=NormalSphere_Location, defaultValue=<rgba 0 0 160 1>}.RGB)
  n10 = Multiply(A=@n9, B=Constant3Vector() {constant=<rgba 1 1 0 0>})
  n11 = ScalarParameter() {defaultValue=1, group=Toon_Shading, parameterName=NormalSphereCylindrical_Blend, sliderMax=1}
  n12 = Transform(None=Normalize(VectorInput=LinearInterpolate(A=@n9, B=@n10, Alpha=@n11))) {transformSourceType=TRANSFORMSOURCE_Local}
  n13 = LinearInterpolate(
   A=@n2
   B=LinearInterpolate(
    A=@n2
    B=VertexInterpolator(VS=@n12).PS
    Alpha=ScalarParameter() {defaultValue=0.75, group=Toon_Shading, parameterName=NormalSphere_Blend, sliderMax=1}
   ) {constAlpha=0}
   Alpha=@n7
  )
  n14 = Multiply(A=@n4) {constB=2}
  n15 = ScalarParameter() {desc="0.0 = surface normals, 1.0 baked vertex color normal", group=02_Anime Parameters, parameterName=VertexBakedNormal_Blend, sliderMax=1}
  n16 = LinearInterpolate(A=@n2, B=Transform(None=Subtract(A=@n14)) {transformSourceType=TRANSFORMSOURCE_Local}, Alpha=Multiply(A=@n15, B=@n4.A))
  n17 = MaterialFunctionCall!(
   Normal_WS=StaticSwitchParameter(
    True=StaticSwitchParameter(True=@n8, False=@n13) {group=Toon_Shading, parameterName=UseCapsuleTangentNormal}
    False=Normalize(VectorInput=@n16)
   ) {group=Toon_Shading, parameterName=UseSphereCylindricalNormal}
   Color_Lit=@n1.Color_Lit
   Color_Shaded=@n1.Color_Lit
  ) {materialFunction=&/MF_Superlevel_Sim}
  n18 = MaterialFunctionCall!(
   CustomLightVector=@n17.CustomeLightVector
   OverrideNormal_WS=@n17.CustomeLightVector
   STT=@n1.STT
   Normal_Map=@n17.CustomeLightVector
   DFL=@n1.STT
   FX=@n1.STT
  ) {materialFunction=/Game/Global/Materials/Character_Parent_2023/Anime_Parent_2023/MaterialFunctions/MFC_AnimeShader_2023_Calculations}
  MaterialFunctionCall!(
   Shading_Gradient=@n18.Shading_Gradient
   Threshold_Gradient=@n18.Shading_Gradient
   Specular_Gradient=@n18.Shading_Gradient
   Sheen_Mask=@n17.Sheen_Mask
   Thickness=@n18.Shading_Gradient
   DFL_Mask=MaterialFunctionCall!(DFLines=@n18.DFL_Mask) {materialFunction=/BRCosmetics/Characters/Player/Male/Medium/Bodies/M_MED_SilentNovel_Vane/Materials/MF_EyeballsDFLineMasking}.Result
   FX=@n18.Shading_Gradient
   Metallic_Gradient=@n18.Shading_Gradient
   RimLight_Gradient=@n18.Shading_Gradient
   NormalBuffer=@n18.Shading_Gradient
   --------
   CINE_LightSources=@n18.Shading_Gradient
   CINE_LightColor=@n18.Shading_Gradient
   ---------
   Color_Lit=@n17.Sheen_Mask
   Color_Shaded=@n17.Sheen_Mask
   SRM=@n1.SRM
   Emissive=@n1.SRM
   DFL_Color=@n1.SRM
   Specular_Color=@n1.SRM
   RimLight_Color=@n1.SRM
  ) {materialFunction=/Game/Global/Materials/Character_Parent_2023/Anime_Parent_2023/MaterialFunctions/MFC_AnimeShader_2023_Post}
 ]
}
